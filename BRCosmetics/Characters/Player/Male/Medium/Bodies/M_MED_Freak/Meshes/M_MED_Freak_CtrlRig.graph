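# Control Rig for the M_MED_Freak player body, applying corrective offsets to skirt, vial, frog and worm accessory bones.
# Thigh-to-pelvis angles are remapped into weights for additive offsets, with crouching suppressing the frog and chest vial correction.
# Crouching, targeting and jumping variables weight leg vial, neck/head and worm adjustments respectively.
# Parent constraints blend worm and chest-tie bases, and ChainHarmonics waves animate two worm chains.
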ControlRigBlueprint /BRCosmetics/Characters/Player/Male/Medium/Bodies/M_MED_Freak/Meshes/M_MED_Freak_CtrlRig
graphs [
 RigVMFunctionLibrary {links [], nodes []}
 RigVMModel {nodes [
   n1 = &RigUnit_BeginExecution()
   n2 = &RigUnit_GetRelativeBoneTransform(Bone=thigh_l, Space=pelvis)
   n3 = &RigVMFunction_MathFloatRemap(
    SourceMaximum=106
    SourceMinimum=20
    TargetMinimum=0
    bClamp=true
    Value=/Script/RigVM.RigVMFunction_MathFloatAbs(Value=&RigVMFunction_MathQuaternionToEuler(RotationOrder=ZYX, Value=@n2.Transform.Rotation).Result.Z) {is=struct RigVMUnitNode, template="Absolute::Execute(in Value,out Result)"}.Result
   )
   /Script/ControlRig.RigUnit_ModifyBoneTransforms(
    BoneToModify=((Bone="dyn_skirtFront_01_Base_l",Transform=(Rotation=(X=0,Y=0,Z=0,W=1),Translation=(X=-12,Y=-7.5,Z=2),Scale3D=(X=1,Y=1,Z=1))),(Bone="None",Transform=(Rotation=(X=0,Y=0,Z=0,W=1),Translation=(X=0,Y=0,Z=1),Scale3D=(X=1,Y=1,Z=1))))
    Mode=AdditiveLocal
    WeightMaximum=1
    WeightMinimum=0
    ExecutePin=@n1.ExecutePin
    Weight=@n3.Result
    BoneToModify.0.Transform.Rotation=&RigVMFunction_MathQuaternionFromEuler(Euler=(X=45,Y=0,Z=0)).Result
   ) {is=struct RigVMUnitNode}
   > &RigUnit_ModifyBoneTransforms(
    BoneToModify=((Bone="dyn_skirtFrontSide_01_Base_l",Transform=(Rotation=(X=0,Y=0,Z=0,W=1),Translation=(X=-8,Y=-3,Z=0),Scale3D=(X=1,Y=1,Z=1))),(Bone="None",Transform=(Rotation=(X=0,Y=0,Z=0,W=1),Translation=(X=0,Y=0,Z=1),Scale3D=(X=1,Y=1,Z=1))))
    Mode=AdditiveLocal
    WeightMaximum=1
    WeightMinimum=0
    Weight=@n3.Result
   )
   > &RigUnit_ModifyBoneTransforms(
    BoneToModify=((Bone="dyn_skirtSide_01_Base_l",Transform=(Rotation=(X=0,Y=0,Z=0,W=1),Translation=(X=-2,Y=0,Z=1),Scale3D=(X=1,Y=1,Z=1))),(Bone="None",Transform=(Rotation=(X=0,Y=0,Z=0,W=1),Translation=(X=0,Y=0,Z=1),Scale3D=(X=1,Y=1,Z=1))))
    Mode=AdditiveLocal
    WeightMaximum=1
    WeightMinimum=0
    Weight=@n3.Result
   )
   n4 = &RigUnit_GetRelativeBoneTransform(Bone=thigh_r, Space=pelvis)
   n5 = &RigVMFunction_MathFloatRemap(
    SourceMaximum=75
    SourceMinimum=160
    TargetMinimum=0
    bClamp=true
    Value=&RigVMFunction_MathFloatAbs(Value=&RigVMFunction_MathQuaternionToEuler(RotationOrder=ZYX, Value=@n4.Transform.Rotation).Result.Z) {template="Absolute::Execute(in Value,out Result)"}.Result
   )
   > &RigUnit_ModifyBoneTransforms(
    BoneToModify=((Bone="dyn_skirtFront_01_Base_r",Transform=(Rotation=(X=0,Y=0,Z=0,W=1),Translation=(X=8,Y=2,Z=2),Scale3D=(X=1,Y=1,Z=1))),(Bone="None",Transform=(Rotation=(X=0,Y=0,Z=0,W=1),Translation=(X=0,Y=0,Z=1),Scale3D=(X=1,Y=1,Z=1))))
    Mode=AdditiveLocal
    WeightMaximum=1
    WeightMinimum=0
    Weight=@n5.Result
   )
   > &RigUnit_ModifyBoneTransforms(
    BoneToModify=((Bone="dyn_skirtFront_01_r",Transform=(Rotation=(X=0,Y=0,Z=0,W=1),Translation=(X=2,Y=2,Z=0),Scale3D=(X=1,Y=1,Z=1))),(Bone="None",Transform=(Rotation=(X=0,Y=0,Z=0,W=1),Translation=(X=0,Y=0,Z=1),Scale3D=(X=1,Y=1,Z=1))))
    Mode=AdditiveLocal
    WeightMaximum=1
    WeightMinimum=0
    Weight=@n5.Result
   )
   > &RigUnit_ModifyBoneTransforms(
    BoneToModify=((Bone="dyn_skirtFrontSide_02_r",Transform=(Rotation=(X=0,Y=0,Z=0,W=1),Translation=(X=-4,Y=-3,Z=1),Scale3D=(X=1,Y=1,Z=1))),(Bone="None",Transform=(Rotation=(X=0,Y=0,Z=0,W=1),Translation=(X=0,Y=0,Z=1),Scale3D=(X=1,Y=1,Z=1))))
    Mode=AdditiveLocal
    WeightMaximum=1
    WeightMinimum=0
    Weight=@n5.Result
   )
   > &RigUnit_ModifyBoneTransforms(
    BoneToModify=((Bone="dyn_skirtFrontSide_01_Base_r",Transform=(Rotation=(X=0,Y=0,Z=0,W=1),Translation=(X=15,Y=3,Z=0),Scale3D=(X=1,Y=1,Z=1))),(Bone="None",Transform=(Rotation=(X=0,Y=0,Z=0,W=1),Translation=(X=0,Y=0,Z=1),Scale3D=(X=1,Y=1,Z=1))))
    Mode=AdditiveLocal
    WeightMaximum=1
    WeightMinimum=0
    Weight=@n5.Result
   )
   > &RigUnit_ParentConstraint(
    AdvancedSettings=(InterpolationType=Average,RotationOrderForFilter=XZY)
    Child=(Type=Bone,Name="dyn_wormAll_Base")
    Parents=((Item=(Type=Bone,Name="pelvis"),Weight=0.4),(Item=(Type=Bone,Name="dyn_skirtFront_01_Base_r"),Weight=0.6))
    Weight=1
   )
   > &RigUnit_ModifyBoneTransforms(
    BoneToModify=((Bone="dyn_skirtSide_01_Base_r",Transform=(Rotation=(X=0,Y=0,Z=0,W=1),Translation=(X=0,Y=0,Z=0),Scale3D=(X=1,Y=1,Z=1))),(Bone="None",Transform=(Rotation=(X=0,Y=0,Z=0,W=1),Translation=(X=0,Y=0,Z=1),Scale3D=(X=1,Y=1,Z=1))))
    Mode=AdditiveLocal
    Weight=1
    WeightMaximum=1
    WeightMinimum=0
   )
   > &RigUnit_ModifyBoneTransforms(
    BoneToModify=((Bone="dyn_wormAll_Base",Transform=(Rotation=(X=0,Y=0,Z=0,W=1),Translation=(X=1,Y=0,Z=-7),Scale3D=(X=1,Y=1,Z=1))),(Bone="None",Transform=(Rotation=(X=0,Y=0,Z=0,W=1),Translation=(X=0,Y=0,Z=1),Scale3D=(X=1,Y=1,Z=1))))
    Mode=AdditiveLocal
    WeightMaximum=1
    WeightMinimum=0
    Weight=@n5.Result
    BoneToModify.0.Transform.Rotation=&RigVMFunction_MathQuaternionFromEuler(Euler=(X=-25,Y=10,Z=0)).Result
   )
   > &RigUnit_ParentConstraint(
    AdvancedSettings=(InterpolationType=Average,RotationOrderForFilter=XZY)
    Child=(Type=Bone,Name="dyn_chestTieBottom_Base")
    Parents=((Item=(Type=Bone,Name="dyn_skirtFront_01_r"),Weight=0.4),(Item=(Type=Bone,Name="dyn_skirtFront_01_l"),Weight=0.6))
    Weight=1
   )
   > /Script/ControlRig.RigUnit_ChainHarmonics(
    ChainRoot=dyn_wormA_01
    DrawWorldOffset=(Rotation=(X=0,Y=0,Z=0,W=1),Translation=(X=0,Y=0,Z=0),Scale3D=(X=1,Y=1,Z=1))
    Pendulum=(bEnabled=False,PendulumStiffness=2,PendulumGravity=(X=0,Y=0,Z=0),PendulumBlend=0.75,PendulumDrag=0.98,PendulumMinimum=0,PendulumMaximum=0.1,PendulumEase=Linear,UnwindAxis=(X=0,Y=1,Z=0),UnwindMinimum=0.2,UnwindMaximum=0.05)
    Reach=(bEnabled=False,ReachTarget=(X=0,Y=0,Z=0),ReachAxis=(X=1,Y=0,Z=0),ReachMinimum=0,ReachMaximum=0,ReachEase=SineEaseInOut)
    Speed=(X=1.1,Y=0.6,Z=0.45)
    Wave=(bEnabled=True,WaveFrequency=(X=4,Y=4,Z=4),WaveAmplitude=(X=1,Y=1,Z=2),WaveOffset=(X=0,Y=10,Z=20),WaveNoise=(X=2,Y=3,Z=3),WaveMinimum=0,WaveMaximum=0.75,WaveEase=SineEaseIn)
    WaveCurve=(EditorCurveData=(Keys=((Time=-0.0198,Value=0.4929),(Time=0.9855,Value=0.5049)),PreInfinityExtrap=RCCE_Constant,PostInfinityExtrap=RCCE_Constant,DefaultValue=340282346638528859811704183484516925440))
    bDrawDebug=True
   ) {is=struct RigVMUnitNode}
   > &RigUnit_ChainHarmonics(
    ChainRoot=dyn_wormB_01
    DrawWorldOffset=(Rotation=(X=0,Y=0,Z=0,W=1),Translation=(X=0,Y=0,Z=0),Scale3D=(X=1,Y=1,Z=1))
    Pendulum=(bEnabled=False,PendulumStiffness=2,PendulumGravity=(X=0,Y=0,Z=0),PendulumBlend=0.75,PendulumDrag=0.98,PendulumMinimum=0,PendulumMaximum=0.1,PendulumEase=Linear,UnwindAxis=(X=0,Y=1,Z=0),UnwindMinimum=0.2,UnwindMaximum=0.05)
    Reach=(bEnabled=False,ReachTarget=(X=0,Y=0,Z=0),ReachAxis=(X=1,Y=0,Z=0),ReachMinimum=0,ReachMaximum=0,ReachEase=SineEaseInOut)
    Speed=(X=1.1,Y=0.6,Z=0.45)
    Wave=(bEnabled=True,WaveFrequency=(X=1,Y=4,Z=2),WaveAmplitude=(X=1,Y=1,Z=2),WaveOffset=(X=0,Y=10,Z=15),WaveNoise=(X=2.5,Y=4,Z=1),WaveMinimum=0,WaveMaximum=0.75,WaveEase=SineEaseIn)
    WaveCurve=(EditorCurveData=(Keys=((Time=-0.0198,Value=0.4929),(Time=0.9855,Value=0.5049)),PreInfinityExtrap=RCCE_Constant,PostInfinityExtrap=RCCE_Constant,DefaultValue=340282346638528859811704183484516925440))
    bDrawDebug=True
   )
   n6 = &RigUnit_GetRelativeBoneTransform(Bone=thigh_l, Space=pelvis)
   n7 = &RigVMFunction_MathFloatRemap(
    SourceMaximum=105
    SourceMinimum=40
    TargetMinimum=0
    bClamp=true
    Value=&RigVMFunction_MathFloatAbs(Value=&RigVMFunction_MathQuaternionToEuler(RotationOrder=ZYX, Value=@n6.Transform.Rotation).Result.Z) {template="Absolute::Execute(in Value,out Result)"}.Result
   )
   n8 = RigVMVariableNode() {variable=crouching}
   n9 = &RigVMFunction_MathFloatMul(
    A=@n7.Result
    B=/Script/RigVM.RigVMFunction_MathFloatSub(A=1, B=@n8.Value) {is=struct RigVMUnitNode, template="Subtract::Execute(in A,in B,out Result)"}.Result
   ) {template="Multiply::Execute(in A,in B,out Result)"}
   > &RigUnit_ModifyBoneTransforms(
    BoneToModify=((Bone="dyn_frogMain_Base",Transform=(Rotation=(X=0,Y=0,Z=0,W=1),Translation=(X=0,Y=-10,Z=0),Scale3D=(X=1,Y=1,Z=1))),(Bone="None",Transform=(Rotation=(X=0,Y=0,Z=0,W=1),Translation=(X=0,Y=0,Z=1),Scale3D=(X=1,Y=1,Z=1))))
    Mode=AdditiveLocal
    WeightMaximum=1
    WeightMinimum=0
    Weight=@n9.Result
    BoneToModify.0.Transform.Rotation=&RigVMFunction_MathQuaternionFromEuler(Euler=(X=0,Y=60,Z=-45)).Result
   )
   > &RigUnit_ModifyBoneTransforms(
    BoneToModify=((Bone="dyn_chestVial_Base",Transform=(Rotation=(X=0,Y=0,Z=0,W=1),Translation=(X=-2,Y=-2,Z=2),Scale3D=(X=1,Y=1,Z=1))),(Bone="None",Transform=(Rotation=(X=0,Y=0,Z=0,W=1),Translation=(X=0,Y=0,Z=1),Scale3D=(X=1,Y=1,Z=1))))
    Mode=AdditiveLocal
    WeightMaximum=1
    WeightMinimum=0
    Weight=@n9.Result
    BoneToModify.0.Transform.Rotation=&RigVMFunction_MathQuaternionFromEuler(Euler=(X=0,Y=0,Z=-45)).Result
   )
   > &RigUnit_ModifyBoneTransforms(
    BoneToModify=((Bone="dyn_legVialA_Base",Transform=(Rotation=(X=0,Y=0,Z=0,W=1),Translation=(X=0,Y=0,Z=0),Scale3D=(X=1,Y=1,Z=1))),(Bone="None",Transform=(Rotation=(X=0,Y=0,Z=0,W=1),Translation=(X=0,Y=0,Z=1),Scale3D=(X=1,Y=1,Z=1))))
    Mode=AdditiveLocal
    WeightMaximum=1
    WeightMinimum=0
    BoneToModify.0.Transform.Rotation=&RigVMFunction_MathQuaternionFromEuler(Euler=(X=15,Y=55,Z=-30)).Result
    Weight=@n8.Value
   )
   n10 = &RigVMFunction_MathQuaternionFromEuler(Euler=(X=0,Y=30,Z=-8))
   > &RigUnit_ModifyBoneTransforms(
    BoneToModify=((Bone="dyn_legVialB_Base",Transform=(Rotation=(X=0,Y=0,Z=0,W=1),Translation=(X=0,Y=0,Z=0),Scale3D=(X=1,Y=1,Z=1))),(Bone="None",Transform=(Rotation=(X=0,Y=0,Z=0,W=1),Translation=(X=0,Y=0,Z=1),Scale3D=(X=1,Y=1,Z=1))))
    Mode=AdditiveLocal
    WeightMaximum=1
    WeightMinimum=0
    BoneToModify.0.Transform.Rotation=@n10.Result
    Weight=@n8.Value
   )
   > &RigUnit_ModifyBoneTransforms(
    BoneToModify=((Bone="dyn_legVialC_Base",Transform=(Rotation=(X=0,Y=0,Z=0,W=1),Translation=(X=0,Y=0,Z=0),Scale3D=(X=1,Y=1,Z=1))),(Bone="None",Transform=(Rotation=(X=0,Y=0,Z=0,W=1),Translation=(X=0,Y=0,Z=1),Scale3D=(X=1,Y=1,Z=1))))
    Mode=AdditiveLocal
    WeightMaximum=1
    WeightMinimum=0
    BoneToModify.0.Transform.Rotation=@n10.Result
    Weight=@n8.Value
   )
   n11 = &RigVMFunction_MathQuaternionFromEuler(Euler=(X=-10,Y=0,Z=0))
   n12 = RigVMVariableNode() {variable=targeting}
   > &RigUnit_ModifyBoneTransforms(
    BoneToModify=((Bone="neck_01",Transform=(Rotation=(X=0,Y=0,Z=0,W=1),Translation=(X=0,Y=0,Z=0),Scale3D=(X=1,Y=1,Z=1))))
    Mode=AdditiveLocal
    WeightMaximum=1
    WeightMinimum=0
    BoneToModify.0.Transform.Rotation=@n11.Result
    Weight=@n12.Value
   )
   > &RigUnit_ModifyBoneTransforms(
    BoneToModify=((Bone="neck_02",Transform=(Rotation=(X=0,Y=0,Z=0,W=1),Translation=(X=0,Y=0,Z=0),Scale3D=(X=1,Y=1,Z=1))))
    Mode=AdditiveLocal
    WeightMaximum=1
    WeightMinimum=0
    BoneToModify.0.Transform.Rotation=@n11.Result
    Weight=@n12.Value
   )
   > &RigUnit_ModifyBoneTransforms(
    BoneToModify=((Bone="head",Transform=(Rotation=(X=0,Y=0,Z=0,W=1),Translation=(X=0,Y=0,Z=0),Scale3D=(X=1,Y=1,Z=1))))
    Mode=AdditiveLocal
    WeightMaximum=1
    WeightMinimum=0
    BoneToModify.0.Transform.Rotation=@n11.Result
    Weight=@n12.Value
   )
   > &RigUnit_ModifyBoneTransforms(
    BoneToModify=((Bone="dyn_wormC_01_Base",Transform=(Rotation=(X=0,Y=0,Z=0,W=1),Translation=(X=-1,Y=-2,Z=-2),Scale3D=(X=1,Y=1,Z=1))))
    Mode=AdditiveLocal
    WeightMaximum=1
    WeightMinimum=0
    Weight=RigVMVariableNode() {variable=jumping}.Value
   )
   &RigUnit_ModifyBoneTransforms(
    BoneToModify=((Bone="dyn_frogMain_Base",Transform=(Rotation=(X=0,Y=0,Z=0,W=1),Translation=(X=-3.5,Y=1,Z=-10),Scale3D=(X=1,Y=1,Z=1))),(Bone="None",Transform=(Rotation=(X=0,Y=0,Z=0,W=1),Translation=(X=0,Y=0,Z=1),Scale3D=(X=1,Y=1,Z=1))))
    ExecutePin=()
    Mode=AdditiveLocal
    Weight=1
    WeightMaximum=1
    WeightMinimum=0
    BoneToModify.0.Transform.Rotation=&RigVMFunction_MathQuaternionFromEuler(Euler=(X=0,Y=50,Z=-10)).Result
   )
   > &RigUnit_ModifyBoneTransforms(
    BoneToModify=((Bone="dyn_chestVial_Base",Transform=(Rotation=(X=0,Y=0,Z=0,W=1),Translation=(X=1,Y=2,Z=-10),Scale3D=(X=1,Y=1,Z=1))),(Bone="None",Transform=(Rotation=(X=0,Y=0,Z=0,W=1),Translation=(X=0,Y=0,Z=1),Scale3D=(X=1,Y=1,Z=1))))
    Mode=AdditiveLocal
    Weight=1
    WeightMaximum=1
    WeightMinimum=0
    BoneToModify.0.Transform.Rotation=&RigVMFunction_MathQuaternionFromEuler(Euler=(X=0,Y=20,Z=20)).Result
   )
   &RigUnit_BeginExecution()
   /Script/RigVM.RigVMFunction_MathFloatLerp(A=0, T=0.8, B=@n7.Result) {is=struct RigVMUnitNode, template="Interpolate::Execute(in A,in B,in T,out Result)"}
  ]}
]
hierarchy [
 root {loc=<0 0 0>, rot=<0 0 0 1>}
 attach {parent=root, loc=<0 0 0>, rot=<0 0 0 1>}
 pelvis {parent=root, loc=<0 -0.7071 93.1191>, rot=<0.0127 0.707 -0.0127 -0.707>}
 spine_01 {parent=pelvis, loc=<6.7494 -0.6388 0>, rot=<0 0 0.034 -0.9994>}
 spine_02 {parent=spine_01, loc=<7.2518 -0.035 0>, rot=<0 0 0.0238 -0.9997>}
 spine_03 {parent=spine_02, loc=<6.0104 0 0>, rot=<0 0 0.0149 -0.9999>}
 spine_04 {parent=spine_03, loc=<8.1024 0 0>, rot=<0 0 -0.0132 -0.9999>}
 spine_05 {parent=spine_04, loc=<10.0146 0 0>, rot=<0 0 -0.0015 -1>}
 clavicle_l {parent=spine_05, loc=<4.3575 -0.1466 -2.3223>, rot=<0.0465 -0.7289 -0.009 -0.6829>}
 upperarm_l {parent=clavicle_l, loc=<15.5111 0 0>, rot=<0.0226 -0.3234 -0.0088 -0.9459>}
 lowerarm_l {parent=upperarm_l, loc=<26.6514 0 0>, rot=<0.0391 -0.0099 -0.2453 -0.9686>}
 hand_l {parent=lowerarm_l, loc=<22.3491 0 0>, rot=<-0.7081 -0.0507 0.0975 -0.6975>}
 index_metacarpal_l {parent=hand_l, loc=<1.6535 0.0183 -2.96>, rot=<0.0158 -0.0124 -0.0173 -0.9996>}
 index_01_l {parent=index_metacarpal_l, loc=<8.3008 0.0357 -0.0524>, rot=<-0.0044 0.0485 0.0924 -0.9945>}
 index_02_l {parent=index_01_l, loc=<3.3587 -0.3004 0.0452>, rot=<-0.002 0.0027 0.3225 -0.9466>}
 index_03_l {parent=index_02_l, loc=<2.5578 -0.123 -0.0252>, rot=<-0.002 0.0067 0.2558 -0.9667>}
 middle_metacarpal_l {parent=hand_l, loc=<2.1856 -0.1725 -0.5984>, rot=<0.006 0.0168 -0.0329 -0.9993>}
 middle_01_l {parent=middle_metacarpal_l, loc=<7.3275 -0.2447 0.0129>, rot=<-0.0208 0.0042 0.0858 -0.9961>}
 middle_02_l {parent=middle_01_l, loc=<4.0123 0.1401 0.1175>, rot=<-0.0218 0.0258 0.3512 -0.9357>}
 middle_03_l {parent=middle_02_l, loc=<2.6077 0.0534 0.1373>, rot=<0.0082 0.0188 0.1708 -0.9851>}
 pinky_metacarpal_l {parent=hand_l, loc=<1.7616 -0.9753 3.3814>, rot=<-0.0397 0.0814 -0.069 -0.9935>}
 pinky_01_l {parent=pinky_metacarpal_l, loc=<7.1709 -0.4831 -0.1094>, rot=<-0.0654 -0.0035 0.2242 -0.9723>}
 pinky_02_l {parent=pinky_01_l, loc=<2.8274 0.0893 -0.1043>, rot=<0.0037 0.0011 0.2854 -0.9584>}
 pinky_03_l {parent=pinky_02_l, loc=<2.2046 0.0307 -0.0078>, rot=<-0.0002 0.0098 0.1445 -0.9895>}
 ring_metacarpal_l {parent=hand_l, loc=<2.2074 -0.3625 1.3638>, rot=<-0.0284 0.0722 -0.0458 -0.9959>}
 ring_01_l {parent=ring_metacarpal_l, loc=<7.0295 -0.6197 -0.1446>, rot=<-0.0489 -0.0252 0.1892 -0.9804>}
 ring_02_l {parent=ring_01_l, loc=<3.6174 0.4646 -0.1027>, rot=<0 0 0.2806 -0.9598>}
 ring_03_l {parent=ring_02_l, loc=<2.4152 0.1141 0.0199>, rot=<0 0 0.1189 -0.9929>}
 thumb_01_l {parent=hand_l, loc=<1.3118 -0.3191 -2.459>, rot=<0.6546 -0.4042 0.0664 -0.6354>}
 thumb_02_l {parent=thumb_01_l, loc=<5.9336 -0.0448 0.0401>, rot=<-0.0037 0.0213 0.298 -0.9543>}
 thumb_03_l {parent=thumb_02_l, loc=<2.2223 0.3682 -0.0907>, rot=<0.0118 -0.013 0.0066 -0.9998>}
 weapon_l {parent=hand_l, loc=<-0.1974 -3.411 0.3811>, rot=<0 0 0 -1>}
 lowerarm_twist_01_l {parent=lowerarm_l, loc=<17.8549 0 0>, rot=<0 0 0 1>}
 lowerarm_twist_02_l {parent=lowerarm_l, loc=<9.6199 0 0>, rot=<0 0 0 1>}
 upperarm_twist_01_l {parent=upperarm_l, loc=<0.0041 0 0>, rot=<0 0 0 -1>}
 upperarm_twist_02_l {parent=upperarm_l, loc=<13.7941 0 0>, rot=<0 0 0 -1>}
 dyn_armVialA_Base {parent=upperarm_twist_02_l, loc=<1.1839 4.8099 4.2178>, rot=<0.8417 0.0282 -0.0183 -0.5388>}
 dyn_armVialA {parent=dyn_armVialA_Base, loc=<0 0 0>, rot=<0 0 0 -1>}
 dyn_armVialA_Liquid {parent=dyn_armVialA, loc=<0.7605 -0.0092 -0.0539>, rot=<-0.7877 0.1772 0.3091 -0.5026>}
 dyn_armVialB_Base {parent=upperarm_twist_02_l, loc=<1.2033 1.3192 5.3479>, rot=<0.7034 0.0174 -0.0118 -0.7105>}
 dyn_armVialB {parent=dyn_armVialB_Base, loc=<0 0 0>, rot=<0 0 0 -1>}
 dyn_armVialB_Liquid {parent=dyn_armVialB, loc=<0.7067 -0.0003 0.0018>, rot=<-0.6628 0.2463 0.2514 -0.6609>}
 dyn_armVialC_Base {parent=upperarm_twist_02_l, loc=<1.6471 -2.81 4.6642>, rot=<0.5648 -0.0063 -0.0219 -0.8249>}
 dyn_armVialC {parent=dyn_armVialC_Base, loc=<0 0 0>, rot=<0 0 0 1>}
 dyn_armVialC_Liquid {parent=dyn_armVialC, loc=<0.8722 -0.0078 0.0593>, rot=<-0.5309 0.3073 0.2096 -0.7614>}
 dyn_armVialD_Base {parent=upperarm_twist_02_l, loc=<1.9969 -6.6955 2.1794>, rot=<0.3809 -0.0107 -0.0201 -0.9243>}
 dyn_armVialD {parent=dyn_armVialD_Base, loc=<0 0 0>, rot=<0 0 0 1>}
 dyn_armVialD_Liquid {parent=dyn_armVialD, loc=<1.15 -0.0159 0.0187>, rot=<-0.361 0.3441 0.1411 -0.8552>}
 dyn_armVialE_Base {parent=upperarm_twist_02_l, loc=<2.1072 -8.1169 -2.2451>, rot=<0.0642 -0.0057 -0.0396 -0.9971>}
 dyn_armVialE {parent=dyn_armVialE_Base, loc=<0 0 0>, rot=<0 0 0 1>}
 dyn_armVialE_Liquid {parent=dyn_armVialE, loc=<1.6047 -0.0206 0.027>, rot=<-0.0592 0.3612 0.0458 -0.9295>}
 clavicle_r {parent=spine_05, loc=<4.3575 -0.1466 2.3223>, rot=<0.7289 0.0465 -0.6829 0.009>}
 upperarm_r {parent=clavicle_r, loc=<-15.5111 0 0>, rot=<0.0226 -0.3234 -0.0088 -0.9459>}
 lowerarm_r {parent=upperarm_r, loc=<-26.6514 0 0>, rot=<0.0391 -0.0099 -0.2453 -0.9686>}
 hand_r {parent=lowerarm_r, loc=<-22.3492 0 0>, rot=<-0.7081 -0.0507 0.0975 -0.6975>}
 index_metacarpal_r {parent=hand_r, loc=<-1.6534 -0.0181 2.9601>, rot=<0.0158 -0.0124 -0.0173 -0.9996>}
 index_01_r {parent=index_metacarpal_r, loc=<-8.3008 -0.0357 0.0524>, rot=<-0.0044 0.0485 0.0924 -0.9945>}
 index_02_r {parent=index_01_r, loc=<-3.3589 0.3003 -0.0453>, rot=<-0.002 0.0027 0.3225 -0.9466>}
 index_03_r {parent=index_02_r, loc=<-2.5578 0.1231 0.0251>, rot=<-0.002 0.0067 0.2558 -0.9667>}
 middle_metacarpal_r {parent=hand_r, loc=<-2.1856 0.1726 0.5984>, rot=<0.006 0.0168 -0.0329 -0.9993>}
 middle_01_r {parent=middle_metacarpal_r, loc=<-7.3273 0.2447 -0.0129>, rot=<-0.0208 0.0042 0.0858 -0.9961>}
 middle_02_r {parent=middle_01_r, loc=<-4.0142 -0.1411 -0.07>, rot=<-0.0217 0.0259 0.3512 -0.9357>}
 middle_03_r {parent=middle_02_r, loc=<-2.6074 -0.0534 -0.1312>, rot=<0.0083 0.0188 0.1708 -0.9851>}
 pinky_metacarpal_r {parent=hand_r, loc=<-1.7617 0.9753 -3.3814>, rot=<-0.0397 0.0814 -0.069 -0.9935>}
 pinky_01_r {parent=pinky_metacarpal_r, loc=<-7.1707 0.4832 0.1094>, rot=<-0.0654 -0.0035 0.2242 -0.9723>}
 pinky_02_r {parent=pinky_01_r, loc=<-2.8274 -0.0892 0.1043>, rot=<0.0037 0.0011 0.2854 -0.9584>}
 pinky_03_r {parent=pinky_02_r, loc=<-2.2045 -0.031 0.0078>, rot=<-0.0002 0.0098 0.1445 -0.9895>}
 ring_metacarpal_r {parent=hand_r, loc=<-2.2073 0.3627 -1.3638>, rot=<-0.0284 0.0722 -0.0458 -0.9959>}
 ring_01_r {parent=ring_metacarpal_r, loc=<-7.0297 0.6197 0.1446>, rot=<-0.0489 -0.0252 0.1892 -0.9804>}
 ring_02_r {parent=ring_01_r, loc=<-3.6171 -0.4648 0.1027>, rot=<0 0 0.2806 -0.9598>}
 ring_03_r {parent=ring_02_r, loc=<-2.4153 -0.1141 -0.0199>, rot=<0 0 0.1189 -0.9929>}
 thumb_01_r {parent=hand_r, loc=<-1.3119 0.3192 2.459>, rot=<0.6546 -0.4042 0.0664 -0.6354>}
 thumb_02_r {parent=thumb_01_r, loc=<-5.9335 0.0447 -0.0401>, rot=<-0.0037 0.0213 0.298 -0.9543>}
 thumb_03_r {parent=thumb_02_r, loc=<-2.2223 -0.3682 0.0907>, rot=<0.0118 -0.013 0.0066 -0.9998>}
 weapon_r {parent=hand_r, loc=<0.1975 3.4112 -0.3811>, rot=<0 0 0 -1>}
 lowerarm_twist_01_r {parent=lowerarm_r, loc=<-17.855 0 0>, rot=<0 0 0 1>}
 lowerarm_twist_02_r {parent=lowerarm_r, loc=<-9.6199 0 0>, rot=<0 0 0 1>}
 upperarm_twist_01_r {parent=upperarm_r, loc=<-0.0041 0 0>, rot=<0 0 0 1>}
 upperarm_twist_02_r {parent=upperarm_r, loc=<-13.7942 0 0>, rot=<0 0 0 1>}
 neck_01 {parent=spine_05, loc=<11.3138 0.1529 0>, rot=<0 0 -0.2003 -0.9797>}
 neck_02 {parent=neck_01, loc=<4.5627 0 0>, rot=<0 0 0.0504 -0.9987>}
 head {parent=neck_02, loc=<5.3914 0 0>, rot=<0 0 0.115 -0.9934>}
 dyn_hood_01_Base {parent=head, loc=<9.8083 1.0572 -0.3391>, rot=<0.7894 -0.6103 0.0407 0.0517>}
 dyn_hood_01 {parent=dyn_hood_01_Base, loc=<0 0 0>, rot=<0 0 0 1>}
 dyn_hood_02 {parent=dyn_hood_01, loc=<6.5966 0 0>, rot=<-0.0165 0.1015 -0.085 -0.9911>}
 dyn_hood_03 {parent=dyn_hood_02, loc=<7.554 0 0>, rot=<-0.0044 -0.0618 -0.1455 -0.9874>}
 dyn_tongue_01_Base {parent=spine_05, loc=<10.3139 6.9125 0>, rot=<0 0 -0.8288 -0.5596>}
 dyn_tongue_01 {parent=dyn_tongue_01_Base, loc=<0 0 0>, rot=<0 0 0 -1>}
 dyn_tongue_02 {parent=dyn_tongue_01, loc=<5.9402 0 0>, rot=<0 0 -0.2765 -0.961>}
 dyn_tongue_03 {parent=dyn_tongue_02, loc=<8.0944 0 0>, rot=<0 0 -0.1669 -0.986>}
 dyn_tongue_04 {parent=dyn_tongue_03, loc=<7.5277 0 0>, rot=<0 0 -0.0768 -0.997>}
 dyn_tongue_05 {parent=dyn_tongue_04, loc=<4.3252 0 0>, rot=<0 0 -0.0072 -1>}
 dyn_tongue_06 {parent=dyn_tongue_05, loc=<4.3508 0 0>, rot=<0 0 -0.0314 -0.9995>}
 dyn_tongue_07 {parent=dyn_tongue_06, loc=<4.3883 0 0>, rot=<0 0 -0.0033 -1>}
 dyn_tongue_08 {parent=dyn_tongue_07, loc=<4.1829 0 0>, rot=<0 0 0 1>}
 dyn_chestBoneB_Base {parent=spine_02, loc=<3.0245 16.015 0.2524>, rot=<-0.0252 0.0359 0.9826 0.1804>}
 dyn_chestBoneB {parent=dyn_chestBoneB_Base, loc=<0 0 0>, rot=<0 0 0 1>}
 dyn_chestBoneA_Base {parent=spine_02, loc=<5.0655 15.7656 3.9879>, rot=<0.0722 -0.0148 0.9908 0.1138>}
 dyn_chestBoneA {parent=dyn_chestBoneA_Base, loc=<0 0 0>, rot=<0 0 0 1>}
 dyn_chestVial_Base {parent=spine_01, loc=<2.7547 14.4996 -10.6117>, rot=<0.06 0.2082 0.9534 0.2101>}
 dyn_chestVial {parent=dyn_chestVial_Base, loc=<0 0 0>, rot=<0 0 0 1>}
 dyn_chestVial_Liquid {parent=dyn_chestVial, loc=<6.0027 -0.7602 0.273>, rot=<-0.6269 -0.0129 0.7165 -0.3059>}
 thigh_l {parent=pelvis, loc=<-2.5693 -0.0109 -7.7501>, rot=<0.0012 0.0572 0.0155 -0.9982>}
 dyn_skirtSide_01_Base_l {parent=thigh_l, loc=<-1.0203 4.8177 -10.4068>, rot=<0.5933 0.0152 0.0129 -0.8048>}
 dyn_skirtSide_01_l {parent=dyn_skirtSide_01_Base_l, loc=<0 0 0>, rot=<0 0 0 1>}
 dyn_skirtSide_02_l {parent=dyn_skirtSide_01_l, loc=<-7.2938 0 0>, rot=<0.0386 -0.0248 -0.1737 -0.9837>}
 dyn_skirtSide_03_l {parent=dyn_skirtSide_02_l, loc=<-12.1349 0.7331 0.2678>, rot=<0.4957 0.5043 0.3851 0.5931>}
 dyn_skirtFrontSide_01_Base_l {parent=thigh_l, loc=<-0.8965 11.5736 -5.3193>, rot=<0.3782 -0.0062 0.0297 -0.9252>}
 dyn_skirtFrontSide_01_l {parent=dyn_skirtFrontSide_01_Base_l, loc=<0 0 0>, rot=<0 0 -0.0175 -0.9998>}
 dyn_skirtFrontSide_02_l {parent=dyn_skirtFrontSide_01_l, loc=<-6.9304 0 -0.0001>, rot=<-0.0531 0 -0.1307 -0.99>}
 dyn_skirtFrontSide_03_l {parent=dyn_skirtFrontSide_02_l, loc=<-12.1294 0.6639 -0.7159>, rot=<0.2969 0.6597 0.1689 0.6694>}
 dyn_skirtFront_01_Base_l {parent=thigh_l, loc=<0.2807 16.5643 2.9937>, rot=<0.0559 0.0286 -0.0309 -0.9976>}
 dyn_skirtFront_01_l {parent=dyn_skirtFront_01_Base_l, loc=<0 0 0>, rot=<0 0 0 -1>}
 dyn_skirtFront_02_l {parent=dyn_skirtFront_01_l, loc=<-8.1314 0 0>, rot=<0.012 -0.0544 -0.0701 -0.996>}
 dyn_skirtFront_03_l {parent=dyn_skirtFront_02_l, loc=<-12.2611 -0.0448 -1.0192>, rot=<0.122 0.721 -0.013 0.682>}
 calf_l {parent=thigh_l, loc=<-39.0484 0 0>, rot=<0.0006 -0.0161 -0.0391 -0.9991>}
 calf_twist_01_l {parent=calf_l, loc=<-35.8131 -0.0967 0.0397>, rot=<0.0237 0.0095 -0.0035 -0.9997>}
 calf_twist_02_l {parent=calf_l, loc=<-18.8064 -0.0967 0.0397>, rot=<0.0237 0.0095 -0.0035 -0.9997>}
 foot_l {parent=calf_l, loc=<-42.561 0.1077 0.4292>, rot=<0.0418 -0.0455 0.0343 -0.9975>}
 ball_l {parent=foot_l, loc=<-6.1599 13.8264 0.6806>, rot=<0.0001 -0.0001 -0.7063 -0.7079>}
 thigh_twist_01_l {parent=thigh_l, loc=<-21.6303 0 0>, rot=<0.0013 0 0 -1>}
 thigh_r {parent=pelvis, loc=<-2.5691 -0.0109 7.7501>, rot=<0.0572 -0.0012 0.9982 0.0155>}
 dyn_skirtSide_01_Base_r {parent=thigh_r, loc=<0.9684 -5.2154 11.6906>, rot=<0.5933 0.0152 0.0129 -0.8048>}
 dyn_skirtSide_01_r {parent=dyn_skirtSide_01_Base_r, loc=<0 0 0>, rot=<0 0 0 1>}
 dyn_skirtSide_02_r {parent=dyn_skirtSide_01_r, loc=<7.2939 0 0>, rot=<0.0386 -0.0248 -0.1737 -0.9837>}
 dyn_skirtSide_03_r {parent=dyn_skirtSide_02_r, loc=<12.8759 0 0>, rot=<-0.5931 -0.385 0.5043 0.4957>}
 dyn_skirtFrontSide_01_Base_r {parent=thigh_r, loc=<0.7967 -12.992 6.714>, rot=<0.3782 -0.0062 0.0297 -0.9252>}
 dyn_skirtFrontSide_01_r {parent=dyn_skirtFrontSide_01_Base_r, loc=<0 0 0>, rot=<0 0 0 -1>}
 dyn_skirtFrontSide_02_r {parent=dyn_skirtFrontSide_01_r, loc=<6.9305 0 0>, rot=<-0.0531 -0.0012 -0.1523 -0.9869>}
 dyn_skirtFrontSide_03_r {parent=dyn_skirtFrontSide_02_r, loc=<13.067 0 0>, rot=<-0.6705 -0.1661 0.6581 0.2996>}
 dyn_skirtFront_01_Base_r {parent=thigh_r, loc=<0.0817 -18.7494 -2.7683>, rot=<0.0482 0.0317 -0.0833 -0.9949>}
 dyn_skirtFront_01_r {parent=dyn_skirtFront_01_Base_r, loc=<0 0 0>, rot=<0 0 0 -1>}
 dyn_skirtFront_02_r {parent=dyn_skirtFront_01_r, loc=<8.1313 0 0>, rot=<0.0102 -0.0548 -0.0373 -0.9977>}
 dyn_skirtFront_03_r {parent=dyn_skirtFront_02_r, loc=<12.3887 0 0>, rot=<-0.6768 0.0317 0.7235 0.132>}
 calf_r {parent=thigh_r, loc=<39.0487 0 0>, rot=<0.0006 -0.0161 -0.0391 -0.9991>}
 calf_twist_01_r {parent=calf_r, loc=<35.8132 0.0967 -0.0397>, rot=<0.0237 0.0095 -0.0035 -0.9997>}
 calf_twist_02_r {parent=calf_r, loc=<18.8064 0.0967 -0.0397>, rot=<0.0237 0.0095 -0.0035 -0.9997>}
 foot_r {parent=calf_r, loc=<42.5612 -0.1077 -0.4292>, rot=<0.0418 -0.0455 0.0343 -0.9975>}
 ball_r {parent=foot_r, loc=<6.1598 -13.8264 -0.6806>, rot=<0.0001 -0.0001 -0.7063 -0.7079>}
 thigh_twist_01_r {parent=thigh_r, loc=<21.6303 0 0>, rot=<0.0013 0 0 -1>}
 dyn_legVialA_Base {parent=thigh_twist_01_r, loc=<4.1325 4.2725 7.3615>, rot=<0.7967 0.1304 -0.1516 -0.5703>}
 dyn_legVialA {parent=dyn_legVialA_Base, loc=<0 0 0>, rot=<0 0 0 1>}
 dyn_legVialA_Liquid {parent=dyn_legVialA, loc=<2.1735 0.0955 0.0935>, rot=<0.28 0.493 -0.5125 -0.6449>}
 dyn_legVialB_Base {parent=thigh_twist_01_r, loc=<4.8461 -0.0641 8.203>, rot=<0.7566 0.0508 -0.0218 -0.6516>}
 dyn_legVialB {parent=dyn_legVialB_Base, loc=<0 0 0>, rot=<0 0 0 -1>}
 dyn_legVialB_Liquid {parent=dyn_legVialB, loc=<5.955 -0.3072 0.0929>, rot=<0.3936 0.5496 -0.5206 -0.5215>}
 dyn_legVialC_Base {parent=thigh_twist_01_r, loc=<4.8916 -5.9789 6.8455>, rot=<0.7223 0.0528 -0.0915 -0.6835>}
 dyn_legVialC {parent=dyn_legVialC_Base, loc=<0 0 0>, rot=<0 0 0 1>}
 dyn_legVialC_Liquid {parent=dyn_legVialC, loc=<2.2784 -0.0465 0.0033>, rot=<0.4135 0.4777 -0.5454 -0.5508>}
 dyn_chestTieTop_Base {parent=pelvis, loc=<-0.6047 17.9748 1.4413>, rot=<-0.0127 -0.707 0.0127 -0.707>}
 dyn_chestTieTop {parent=dyn_chestTieTop_Base, loc=<0 0 0>, rot=<0 0 0 1>}
 dyn_chestTieBottom_Base {parent=pelvis, loc=<-7.2534 17.1565 -0.8926>, rot=<-0.0127 -0.707 0.0127 -0.707>}
 dyn_chestTieBottom {parent=dyn_chestTieBottom_Base, loc=<0 0 0>, rot=<0 0 0 1>}
 dyn_frogMain_Base {parent=pelvis, loc=<5.7618 10.9043 -18.1343>, rot=<0.1772 -0.3372 -0.6261 -0.6804>}
 dyn_frogMain {parent=dyn_frogMain_Base, loc=<0 0 0>, rot=<0 0 0 -1>}
 dyn_frogHead {parent=dyn_frogMain, loc=<1.0031 -4.8709 -0.0751>, rot=<-0.0179 0.0084 -0.202 -0.9792>}
 dyn_frogTongue_01 {parent=dyn_frogHead, loc=<3.5212 -3.115 3.6533>, rot=<0.3401 0.1988 -0.3144 -0.8637>}
 dyn_frogTongue_02 {parent=dyn_frogTongue_01, loc=<1.6397 0 0>, rot=<0.01 -0.0704 -0.2103 -0.975>}
 dyn_frogArm_01_r {parent=dyn_frogMain, loc=<2.0443 -2.2586 4.8662>, rot=<0.1173 0.4137 -0.2353 -0.8716>}
 dyn_frogArm_02_r {parent=dyn_frogArm_01_r, loc=<2.7753 0 0>, rot=<0.7046 0.4678 -0.2175 0.4872>}
 dyn_frogArm_03_r {parent=dyn_frogArm_02_r, loc=<4.8543 0 0>, rot=<0.7431 0 0 -0.6691>}
 dyn_frogArm_01_l {parent=dyn_frogMain, loc=<2.0327 -2.1504 -5.1483>, rot=<-0.0505 -0.3737 -0.2924 -0.8788>}
 dyn_frogArm_02_l {parent=dyn_frogArm_01_l, loc=<2.4422 0 0>, rot=<0.7069 0.3928 0.2927 -0.5103>}
 dyn_frogArm_03_l {parent=dyn_frogArm_02_l, loc=<4.8258 0 0>, rot=<-0.788 0 0 -0.6157>}
 dyn_frogBelly {parent=dyn_frogMain, loc=<1.296 2.1005 0>, rot=<0 0 -0.6293 -0.7771>}
 dyn_frogLeg_01_l {parent=dyn_frogBelly, loc=<4.3431 0.4916 -3.4506>, rot=<-0.1776 -0.3115 0.1507 -0.9212>}
 dyn_frogLeg_02_l {parent=dyn_frogLeg_01_l, loc=<3.2675 0 0>, rot=<0.0038 0.0593 -0.6287 -0.7754>}
 dyn_frogLeg_03_l {parent=dyn_frogLeg_02_l, loc=<3.8381 0 0>, rot=<0.1491 0.1944 0.6756 -0.6954>}
 dyn_frogLeg_04_l {parent=dyn_frogLeg_03_l, loc=<4.9618 0 0>, rot=<-0.0306 -0.0332 -0.2519 -0.9667>}
 dyn_frogLeg_05_l {parent=dyn_frogLeg_04_l, loc=<2.7439 0 0>, rot=<-0.5537 0.2872 0.6189 -0.4774>}
 dyn_frogLeg_01_r {parent=dyn_frogBelly, loc=<4.4002 0.3581 3.3911>, rot=<0.2182 0.3195 0.1814 -0.9041>}
 dyn_frogLeg_02_r {parent=dyn_frogLeg_01_r, loc=<3.2446 0 0>, rot=<-0.0618 -0.0505 -0.6398 -0.7644>}
 dyn_frogLeg_03_r {parent=dyn_frogLeg_02_r, loc=<3.8697 0 0>, rot=<-0.1942 -0.1368 0.6825 -0.6912>}
 dyn_frogLeg_04_r {parent=dyn_frogLeg_03_r, loc=<4.9329 0 0>, rot=<0.0811 0.0034 -0.2437 -0.9665>}
 dyn_frogLeg_05_r {parent=dyn_frogLeg_04_r, loc=<2.7609 0 0>, rot=<-0.6422 0.0355 0.7079 -0.2919>}
 dyn_wormAll_Base {parent=pelvis, loc=<2.586 13.2732 10.7737>, rot=<0.4478 -0.5431 0.5315 0.4712>}
 dyn_wormAll {parent=dyn_wormAll_Base, loc=<0 0 0>, rot=<0 0 0 -1>}
 dyn_wormB_01_Base {parent=dyn_wormAll, loc=<1.3938 -0.2191 -0.6567>, rot=<0 0 0 -1>}
 dyn_wormB_01 {parent=dyn_wormB_01_Base, loc=<0 0 0>, rot=<0 0 0 -1>}
 dyn_wormB_02 {parent=dyn_wormB_01, loc=<2.0166 0 0>, rot=<-0.002 -0.0055 -0.1863 -0.9825>}
 dyn_wormB_03 {parent=dyn_wormB_02, loc=<2.0791 0 0>, rot=<-0.1264 -0.1594 -0.1363 -0.9696>}
 dyn_wormB_04 {parent=dyn_wormB_03, loc=<1.8857 0 0>, rot=<-0.1001 -0.1589 0.2409 -0.9522>}
 dyn_wormB_05 {parent=dyn_wormB_04, loc=<1.9271 0 0>, rot=<-0.0011 -0.0562 0.2902 -0.9553>}
 dyn_wormB_06 {parent=dyn_wormB_05, loc=<2.6187 0 0>, rot=<0 0 0 -1>}
 dyn_wormC_01_Base {parent=dyn_wormAll, loc=<2.6393 1.3842 -0.6057>, rot=<-0.021 -0.0412 -0.3096 -0.9498>}
 dyn_wormC_01 {parent=dyn_wormC_01_Base, loc=<0 0 0>, rot=<0 0 0 -1>}
 dyn_wormC_02 {parent=dyn_wormC_01, loc=<2.7846 0 0>, rot=<-0.2984 -0.1724 -0.242 -0.907>}
 dyn_wormC_03 {parent=dyn_wormC_02, loc=<3.3105 0 0>, rot=<0.1607 0.0369 -0.0372 -0.9856>}
 dyn_wormC_04 {parent=dyn_wormC_03, loc=<2.3613 0 0>, rot=<0.1268 0.0536 0.2084 -0.9683>}
 dyn_wormC_05 {parent=dyn_wormC_04, loc=<2.6535 0 0>, rot=<0 0 0 -1>}
 dyn_wormA_01_Base {parent=dyn_wormAll, loc=<1.2994 -2.3542 -0.8412>, rot=<-0.0526 -0.1396 -0.2003 -0.9683>}
 dyn_wormA_01 {parent=dyn_wormA_01_Base, loc=<0 0 0>, rot=<0 0 0 1>}
 dyn_wormA_02 {parent=dyn_wormA_01, loc=<1.9609 0 0>, rot=<-0.2576 -0.6548 0.1336 -0.6979>}
 dyn_wormA_03 {parent=dyn_wormA_02, loc=<2.1813 0 0>, rot=<-0.0234 -0.3706 0.1135 -0.9216>}
 dyn_wormA_04 {parent=dyn_wormA_03, loc=<2.7252 0 0>, rot=<0.0218 -0.1117 0.132 -0.9847>}
 dyn_wormA_05 {parent=dyn_wormA_04, loc=<2.4219 0 0>, rot=<0.0404 -0.1016 0.052 -0.9926>}
 dyn_wormA_06 {parent=dyn_wormA_05, loc=<2.5484 0 0>, rot=<0 0 0 1>}
 ik_foot_root {parent=root, loc=<0 0 0>, rot=<0 0 0 1>}
 ik_foot_l {parent=ik_foot_root, loc=<15.2608 -4.4176 9.4433>, rot=<-0.0357 -0.7045 -0.0284 0.7082>}
 ik_foot_r {parent=ik_foot_root, loc=<-15.2609 -4.4176 9.4431>, rot=<-0.7082 -0.0284 0.7045 -0.0357>}
 ik_hand_root {parent=root, loc=<0 0 0>, rot=<0 0 0 1>}
 ik_hand_gun {parent=ik_hand_root, loc=<-51.973 4.089 102.694>, rot=<0.6084 0.1793 0.4613 -0.6205>}
 ik_hand_l {parent=ik_hand_gun, loc=<53.0284 82.1785 35.2051>, rot=<-0.5102 -0.7906 -0.3387 0>}
 ik_hand_r {parent=ik_hand_gun, loc=<0 0 0>, rot=<0 0 0 -1>}
]
variables [
 crouching {type=double, default=0}
 jumping {type=double, default=0}
 targeting {type=double, default=0}
]
