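# Physics asset for the Freak male medium player body cosmetic, covering its dangling dyn_ accessories: vials, chest bones, a frog and a worm.
# Each accessory hangs from a small Base body with angle-limited constraints; the frog branches into head-tongue, arms and belly-legs, and the worm forms a chain.
# Vials carry nested Liquid bodies, free-rotating on the legs.
# Thigh capsules provide body volume, with leg vial bases constrained to the right thigh twist.
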PhysicsAsset /BRCosmetics/Characters/Player/Male/Medium/Bodies/M_MED_Freak/Meshes/M_MED_Freak_Physics
bodies {
 dyn_armVialA [{shapeType=Capsule, length=9.6251, radius=1.4505, center=<2.3086 -0.0048 -0.0079>, rotation=<rot 0 -90.2202 -89.9162>}]
 dyn_armVialA_Base [{shapeType=Capsule, length=0.505, radius=0.505, rotation=<rot 0 -90 0>}]
 dyn_armVialA_Liquid [{shapeType=Capsule, length=0.5572, radius=0.8355, center=<0 0 -0.4319>}]
 dyn_armVialB [{shapeType=Capsule, length=9.6256, radius=1.4524, center=<2.3082 -0.0117 -0.0072>, rotation=<rot 0 -90.3249 -89.9667>}]
 dyn_armVialB_Base [{shapeType=Capsule, length=0.505, radius=0.505, rotation=<rot 0 -90 0>}]
 dyn_armVialB_Liquid [{shapeType=Capsule, length=0.5572, radius=0.8355, center=<0 0 -0.4319>}]
 dyn_armVialC [{shapeType=Capsule, length=9.6643, radius=1.4478, center=<2.27 -0.0049 -0.0049>, rotation=<rot 0 -90.2653 -89.9171>}]
 dyn_armVialC_Base [{shapeType=Capsule, length=0.505, radius=0.505, rotation=<rot 0 -90 0>}]
 dyn_armVialC_Liquid [{shapeType=Capsule, length=0.5572, radius=0.8355, center=<0 0 -0.4319>}]
 dyn_armVialD [{shapeType=Capsule, length=9.6334, radius=1.4447, center=<2.3002 -0.0182 -0.0005>, rotation=<rot 0 89.3743 -89.9075>}]
 dyn_armVialD_Base [{shapeType=Capsule, length=0.505, radius=0.505, rotation=<rot 0 -90 0>}]
 dyn_armVialD_Liquid [{shapeType=Capsule, length=0.5572, radius=0.8355, center=<0 0 -0.4319>}]
 dyn_armVialE [{shapeType=Capsule, length=9.6251, radius=1.4499, center=<2.3086 -0.0101 -0.0041>, rotation=<rot 0 -90.1806 -89.9572>}]
 dyn_armVialE_Base [{shapeType=Capsule, length=0.505, radius=0.505, rotation=<rot 0 -90 0>}]
 dyn_armVialE_Liquid [{shapeType=Capsule, length=0.5572, radius=0.8355, center=<0 0 -0.4319>}]
 dyn_chestBoneA [{shapeType=Capsule, length=5.4432, radius=1.5369, center=<2.0347 -0.2838 -0.0175>, rotation=<rot 0 -85.5306 -86.4297>}]
 dyn_chestBoneA_Base [{shapeType=Capsule, length=0.505, radius=0.505, rotation=<rot 0 -90 0>}]
 dyn_chestBoneB [{shapeType=Capsule, length=5.5411, radius=1.6922, center=<1.9664 0.4122 -0.0119>, rotation=<rot 0 92.3072 -89.4256>}]
 dyn_chestBoneB_Base [{shapeType=Capsule, length=0.505, radius=0.505, rotation=<rot 0 -90 0>}]
 dyn_chestVial [{shapeType=Capsule, length=7.8491, radius=3.6251, center=<3.4652 -0.205 0.0108>, rotation=<rot 0 90.1632 -89.6114>}]
 dyn_chestVial_Base [{shapeType=Capsule, length=0.505, radius=0.505, rotation=<rot 0 -90 0>}]
 dyn_chestVial_Liquid [{shapeType=Capsule, length=1.1998, radius=1.1767, center=<0 0 -0.4997>}]
 dyn_frogArm_01_l [{shapeType=Capsule, length=1.6369, radius=1.4567, center=<0.8538 0 0>, rotation=<rot 90 0 0>}]
 dyn_frogArm_01_r [{shapeType=Capsule, length=2.3283, radius=1.8471, center=<1.1227 0 0>, rotation=<rot 90 0 0>}]
 dyn_frogArm_02_l [{shapeType=Capsule, length=1.8082, radius=1.269, center=<1.2793 0 -0.0001>, rotation=<rot 90 0 0>}]
 dyn_frogArm_02_r [{shapeType=Capsule, length=1.7484, radius=1.3344, center=<1.4605 0 0>, rotation=<rot 90 0 0>}]
 dyn_frogArm_03_l [{shapeType=Capsule, length=2.9622, radius=2.2557, center=<0.8978 0 0>, rotation=<rot 90 0 0>}]
 dyn_frogArm_03_r [{shapeType=Capsule, length=2.9587, radius=2.2593, center=<1.3945 0 0>, rotation=<rot 90 0 0>}]
 dyn_frogBelly [{shapeType=Capsule, length=0.2505, radius=4.7066, center=<1.3723 0.0001 0>, rotation=<rot 90 0 0>}]
 dyn_frogHead [{shapeType=Capsule, length=4.1398, radius=4.9931, center=<0.6127 -3.0537 0>, rotation=<rot 0.0001 9.8288 89.9995>}]
 dyn_frogLeg_01_l [{shapeType=Capsule, length=3.1271, radius=2.1903, center=<0.2382 -0.4007 -0.0003>, rotation=<rot 90 -165.9637 -133.9532>}]
 dyn_frogLeg_01_r [{shapeType=Capsule, length=3.0377, radius=2.3338, center=<0.0366 -0.5494 0.0001>, rotation=<rot 90 0 27.6939>}]
 dyn_frogLeg_02_l [{shapeType=Capsule, length=2.2965, radius=1.3122, center=<1.0557 0 0>, rotation=<rot 90 0 0>}]
 dyn_frogLeg_02_r [{shapeType=Capsule, length=2.2966, radius=1.3156, center=<1.0007 -0.1865 0.2799>, rotation=<rot 90 0 0>}]
 dyn_frogLeg_03_l [{shapeType=Capsule, length=2.4778, radius=1.5759, center=<0.8929 0.0702 0.2696>, rotation=<rot 90 0 0>}]
 dyn_frogLeg_03_r [{shapeType=Capsule, length=2.4964, radius=1.5445, center=<0.9354 -0.0001 0>, rotation=<rot 90 0 0>}]
 dyn_frogLeg_04_l [{shapeType=Capsule, length=2.2311, radius=1.8521, center=<0.7919 0.0207 0.0964>, rotation=<rot 90 0 0>}]
 dyn_frogLeg_04_r [{shapeType=Capsule, length=2.2181, radius=1.8586, center=<0.798 0 0>, rotation=<rot 90 0 0>}]
 dyn_frogMain [{shapeType=Capsule, length=2.9355, radius=3.821, rotation=<rot 0 0 90>}]
 dyn_frogMain_Base [{shapeType=Capsule, length=0.505, radius=0.505, rotation=<rot 90 0 0>}]
 dyn_frogTongue_01 [{shapeType=Capsule, length=0.7759, radius=0.6374, rotation=<rot 90 0 0>}]
 dyn_frogTongue_02 [{shapeType=Capsule, length=0.9517, radius=0.763, rotation=<rot 90 0 0>}]
 dyn_legVialA [{shapeType=Capsule, length=5.6764, radius=2.9939, center=<2.0271 0.0196 -0.0739>, rotation=<rot 0 89.8562 -89.0993>}]
 dyn_legVialA_Base [{shapeType=Capsule, length=0.505, radius=0.505, rotation=<rot 0 -90 0>}]
 dyn_legVialA_Liquid [{shapeType=Capsule, length=0.5572, radius=0.8355, center=<0 0 -0.4319>}]
 dyn_legVialB [{shapeType=Sphere, radius=3.703, center=<7.5453 -0.0002 0>}, {shapeType=Capsule, length=10.0236, radius=1.4026, rotation=<rot 90 0 0>}]
 dyn_legVialB_Base [{shapeType=Capsule, length=0.505, radius=0.505, rotation=<rot 0 -90 0>}]
 dyn_legVialB_Liquid [{shapeType=Capsule, length=0.5572, radius=0.8355, center=<0 0 -0.4319>}]
 dyn_legVialC [{shapeType=Capsule, length=13.0261, radius=1.5551, center=<4.8551 -0.0379 -0.0012>, rotation=<rot 0 -90.2474 -89.9677>}]
 dyn_legVialC_Base [{shapeType=Capsule, length=0.505, radius=0.505, rotation=<rot 0 -90 0>}]
 dyn_legVialC_Liquid [{shapeType=Capsule, length=0.5572, radius=0.8355, center=<0 0 -0.4319>}]
 dyn_wormC_01 [{shapeType=Capsule, length=1.9435, radius=0.8064, center=<0.9045 -0.2544 -0.2833>, rotation=<rot 0 -83.7194 -88.7658>}]
 dyn_wormC_01_Base [{shapeType=Capsule, length=0.7154, radius=0.7127, center=<-0.1563 0.1257 -0.0447>, rotation=<rot 0 154.0866 -53.0307>}]
 dyn_wormC_02 [{shapeType=Capsule, length=2.3106, radius=0.6664, center=<0.8983 -0.0814 0.0516>, rotation=<rot 0 -87.6068 -87.3059>}]
 dyn_wormC_03 [{shapeType=Capsule, length=2.2719, radius=0.649, center=<0.5711 -0.025 0.0023>, rotation=<rot 0 -88.8285 -87.6101>}]
 dyn_wormC_04 [{shapeType=Capsule, length=2.5395, radius=0.6889, center=<1.2665 0.0522 0.005>, rotation=<rot 0 -91.8747 -87.2291>}]
 thigh_l [{shapeType=Capsule, length=30.4289, radius=8.8583, center=<-17.1405 1.7444 1.406>, rotation=<rot -8.8338 -80.5427 -87.6137>}]
 thigh_twist_01_l [{shapeType=Capsule, length=10.5873, radius=16.8443, center=<12.3488 -1.0322 5.2202>, rotation=<rot 79.2791 7.2858 0>}]
 thigh_twist_01_r [{shapeType=Capsule, length=17.5861, radius=12.2136, center=<-10.9916 -3.7408 0>, rotation=<rot 90 0 0>}]
}
constraints [
 {bone1Name=dyn_armVialA, bone2Name=dyn_armVialA_Base, swing1LimitDegrees=8, swing2LimitDegrees=18, twistLimitDegrees=8}
 {bone1Name=dyn_armVialA_Liquid, bone2Name=dyn_armVialA, swing1LimitDegrees=20, swing2LimitDegrees=20, twistLimitDegrees=20}
 {bone1Name=dyn_armVialB, bone2Name=dyn_armVialB_Base, swing1LimitDegrees=8, swing2LimitDegrees=18, twistLimitDegrees=8}
 {bone1Name=dyn_armVialB_Liquid, bone2Name=dyn_armVialB, swing1LimitDegrees=20, swing2LimitDegrees=20, twistLimitDegrees=20}
 {bone1Name=dyn_armVialC, bone2Name=dyn_armVialC_Base, swing1LimitDegrees=8, swing2LimitDegrees=18, twistLimitDegrees=8}
 {bone1Name=dyn_armVialC_Liquid, bone2Name=dyn_armVialC, swing1LimitDegrees=20, swing2LimitDegrees=20, twistLimitDegrees=20}
 {bone1Name=dyn_armVialD, bone2Name=dyn_armVialD_Base, swing1LimitDegrees=8, swing2LimitDegrees=18, twistLimitDegrees=8}
 {bone1Name=dyn_armVialD_Liquid, bone2Name=dyn_armVialD, swing1LimitDegrees=20, swing2LimitDegrees=20, twistLimitDegrees=20}
 {bone1Name=dyn_armVialE, bone2Name=dyn_armVialE_Base, swing1LimitDegrees=8, swing2LimitDegrees=18, twistLimitDegrees=8}
 {bone1Name=dyn_armVialE_Liquid, bone2Name=dyn_armVialE, swing1LimitDegrees=20, swing2LimitDegrees=20, twistLimitDegrees=20}
 {bone1Name=dyn_legVialA, bone2Name=dyn_legVialA_Base, swing1LimitDegrees=15, swing2LimitDegrees=30, twistLimitDegrees=45, twistMotion=Locked}
 {
  bone1Name=dyn_legVialA_Liquid
  bone2Name=dyn_legVialA
  swing1LimitDegrees=45
  swing1Motion=Free
  swing2LimitDegrees=45
  swing2Motion=Free
  twistLimitDegrees=45
  twistMotion=Free
 }
 {bone1Name=dyn_legVialB, bone2Name=dyn_legVialB_Base, swing1LimitDegrees=15, swing2LimitDegrees=30, twistLimitDegrees=45, twistMotion=Locked}
 {
  bone1Name=dyn_legVialB_Liquid
  bone2Name=dyn_legVialB
  swing1LimitDegrees=45
  swing1Motion=Free
  swing2LimitDegrees=45
  swing2Motion=Free
  twistLimitDegrees=45
  twistMotion=Free
 }
 {bone1Name=dyn_legVialC, bone2Name=dyn_legVialC_Base, swing1LimitDegrees=15, swing2LimitDegrees=30, twistLimitDegrees=45, twistMotion=Locked}
 {
  bone1Name=dyn_legVialC_Liquid
  bone2Name=dyn_legVialC
  swing1LimitDegrees=45
  swing1Motion=Free
  swing2LimitDegrees=45
  swing2Motion=Free
  twistLimitDegrees=45
  twistMotion=Free
 }
 {bone1Name=dyn_frogMain, bone2Name=dyn_frogMain_Base, swing1LimitDegrees=5, swing2LimitDegrees=5, twistLimitDegrees=5}
 {bone1Name=dyn_frogHead, bone2Name=dyn_frogMain, swing1LimitDegrees=15, swing2LimitDegrees=15, twistLimitDegrees=4}
 {bone1Name=dyn_frogTongue_01, bone2Name=dyn_frogHead, swing1LimitDegrees=15, swing2LimitDegrees=15, twistLimitDegrees=4}
 {bone1Name=dyn_frogTongue_02, bone2Name=dyn_frogTongue_01, swing1LimitDegrees=15, swing2LimitDegrees=15, twistLimitDegrees=4}
 {bone1Name=dyn_frogArm_01_r, bone2Name=dyn_frogMain, swing1LimitDegrees=15, swing2LimitDegrees=15, twistLimitDegrees=4}
 {bone1Name=dyn_frogArm_02_r, bone2Name=dyn_frogArm_01_r, swing1LimitDegrees=15, swing2LimitDegrees=15, twistLimitDegrees=4}
 {bone1Name=dyn_frogArm_03_r, bone2Name=dyn_frogArm_02_r, swing1LimitDegrees=15, swing2LimitDegrees=15, twistLimitDegrees=4}
 {bone1Name=dyn_frogArm_01_l, bone2Name=dyn_frogMain, swing1LimitDegrees=15, swing2LimitDegrees=15, twistLimitDegrees=4}
 {bone1Name=dyn_frogArm_02_l, bone2Name=dyn_frogArm_01_l, swing1LimitDegrees=15, swing2LimitDegrees=15, twistLimitDegrees=4}
 {bone1Name=dyn_frogArm_03_l, bone2Name=dyn_frogArm_02_l, swing1LimitDegrees=15, swing2LimitDegrees=15, twistLimitDegrees=4}
 {bone1Name=dyn_frogBelly, bone2Name=dyn_frogMain, swing1LimitDegrees=15, swing2LimitDegrees=15, twistLimitDegrees=4}
 {bone1Name=dyn_frogLeg_01_l, bone2Name=dyn_frogBelly, swing1LimitDegrees=15, swing2LimitDegrees=15, twistLimitDegrees=4}
 {bone1Name=dyn_frogLeg_02_l, bone2Name=dyn_frogLeg_01_l, swing1LimitDegrees=15, swing2LimitDegrees=15, twistLimitDegrees=4}
 {bone1Name=dyn_frogLeg_03_l, bone2Name=dyn_frogLeg_02_l, swing1LimitDegrees=15, swing2LimitDegrees=15, twistLimitDegrees=4}
 {bone1Name=dyn_frogLeg_04_l, bone2Name=dyn_frogLeg_03_l, swing1LimitDegrees=15, swing2LimitDegrees=15, twistLimitDegrees=4}
 {bone1Name=dyn_frogLeg_01_r, bone2Name=dyn_frogBelly, swing1LimitDegrees=15, swing2LimitDegrees=15, twistLimitDegrees=4}
 {bone1Name=dyn_frogLeg_02_r, bone2Name=dyn_frogLeg_01_r, swing1LimitDegrees=15, swing2LimitDegrees=15, twistLimitDegrees=4}
 {bone1Name=dyn_frogLeg_03_r, bone2Name=dyn_frogLeg_02_r, swing1LimitDegrees=15, swing2LimitDegrees=15, twistLimitDegrees=4}
 {bone1Name=dyn_frogLeg_04_r, bone2Name=dyn_frogLeg_03_r, swing1LimitDegrees=15, swing2LimitDegrees=15, twistLimitDegrees=4}
 {
  bone1Name=thigh_twist_01_l
  bone2Name=thigh_l
  swing1LimitDegrees=45
  swing1Motion=Free
  swing2LimitDegrees=45
  swing2Motion=Free
  twistLimitDegrees=45
  twistMotion=Free
 }
 {bone1Name=dyn_wormC_01, bone2Name=dyn_wormC_01_Base, swing1LimitDegrees=16, swing2LimitDegrees=16, twistLimitDegrees=4}
 {bone1Name=dyn_wormC_02, bone2Name=dyn_wormC_01, swing1LimitDegrees=16, swing2LimitDegrees=16, twistLimitDegrees=4}
 {bone1Name=dyn_wormC_03, bone2Name=dyn_wormC_02, swing1LimitDegrees=16, swing2LimitDegrees=16, twistLimitDegrees=4}
 {bone1Name=dyn_wormC_04, bone2Name=dyn_wormC_03, swing1LimitDegrees=16, swing2LimitDegrees=16, twistLimitDegrees=4}
 {
  bone1Name=dyn_legVialA_Base
  bone2Name=thigh_twist_01_r
  swing1LimitDegrees=45
  swing1Motion=Free
  swing2LimitDegrees=45
  swing2Motion=Free
  twistLimitDegrees=45
  twistMotion=Free
 }
 {
  bone1Name=dyn_legVialB_Base
  bone2Name=thigh_twist_01_r
  swing1LimitDegrees=45
  swing1Motion=Free
  swing2LimitDegrees=45
  swing2Motion=Free
  twistLimitDegrees=45
  twistMotion=Free
 }
 {
  bone1Name=dyn_legVialC_Base
  bone2Name=thigh_twist_01_r
  swing1LimitDegrees=45
  swing1Motion=Free
  swing2LimitDegrees=45
  swing2Motion=Free
  twistLimitDegrees=45
  twistMotion=Free
 }
 {bone1Name=dyn_chestBoneB, bone2Name=dyn_chestBoneB_Base, swing1LimitDegrees=16, swing2LimitDegrees=16, twistLimitDegrees=8}
 {bone1Name=dyn_chestBoneA, bone2Name=dyn_chestBoneA_Base, swing1LimitDegrees=16, swing2LimitDegrees=16, twistLimitDegrees=8}
 {bone1Name=dyn_chestVial, bone2Name=dyn_chestVial_Base, swing1LimitDegrees=16, swing2LimitDegrees=16, twistLimitDegrees=8}
 {bone1Name=dyn_chestVial_Liquid, bone2Name=dyn_chestVial, swing1LimitDegrees=12, swing2LimitDegrees=12, twistLimitDegrees=12}
]
