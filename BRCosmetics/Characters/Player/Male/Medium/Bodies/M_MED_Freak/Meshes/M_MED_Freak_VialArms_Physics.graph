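# Physics asset for the vial arms on the M_MED_Freak male player body cosmetic, covering five arm vials (A to E).
# Each vial has three capsule bodies on dyn_ bones: a small base, a long vial body and a short liquid body.
# Each vial is jointed to its base with narrow swing and almost no twist.
# Each liquid body is jointed to its vial with looser, equal swing and twist limits, so it can tilt independently.
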PhysicsAsset /BRCosmetics/Characters/Player/Male/Medium/Bodies/M_MED_Freak/Meshes/M_MED_Freak_VialArms_Physics
bodies {
 dyn_armVialA [{shapeType=Capsule, length=9.6251, radius=1.4505, center=<2.3086 -0.0048 -0.0079>, rotation=<rot 0 -90.2202 -89.9162>}]
 dyn_armVialA_Base [{shapeType=Capsule, length=0.505, radius=0.505, rotation=<rot 0 -90 0>}]
 dyn_armVialA_Liquid [{shapeType=Capsule, length=0.5572, radius=0.8355, center=<0 0 -0.4319>}]
 dyn_armVialB [{shapeType=Capsule, length=9.6256, radius=1.4524, center=<2.3082 -0.0117 -0.0072>, rotation=<rot 0 -90.3249 -89.9667>}]
 dyn_armVialB_Base [{shapeType=Capsule, length=0.505, radius=0.505, rotation=<rot 0 -90 0>}]
 dyn_armVialB_Liquid [{shapeType=Capsule, length=0.5572, radius=0.8355, center=<0 0 -0.4319>}]
 dyn_armVialC [{shapeType=Capsule, length=9.6643, radius=1.4478, center=<2.27 -0.0049 -0.0049>, rotation=<rot 0 -90.2653 -89.9171>}]
 dyn_armVialC_Base [{shapeType=Capsule, length=0.505, radius=0.505, rotation=<rot 0 -90 0>}]
 dyn_armVialC_Liquid [{shapeType=Capsule, length=0.5572, radius=0.8355, center=<0 0 -0.4319>}]
 dyn_armVialD [{shapeType=Capsule, length=9.6334, radius=1.4447, center=<2.3002 -0.0182 -0.0005>, rotation=<rot 0 89.3743 -89.9075>}]
 dyn_armVialD_Base [{shapeType=Capsule, length=0.505, radius=0.505, rotation=<rot 0 -90 0>}]
 dyn_armVialD_Liquid [{shapeType=Capsule, length=0.5572, radius=0.8355, center=<0 0 -0.4319>}]
 dyn_armVialE [{shapeType=Capsule, length=9.6251, radius=1.4499, center=<2.3086 -0.0101 -0.0041>, rotation=<rot 0 -90.1806 -89.9572>}]
 dyn_armVialE_Base [{shapeType=Capsule, length=0.505, radius=0.505, rotation=<rot 0 -90 0>}]
 dyn_armVialE_Liquid [{shapeType=Capsule, length=0.5572, radius=0.8355, center=<0 0 -0.4319>}]
}
constraints [
 {bone1Name=dyn_armVialA, bone2Name=dyn_armVialA_Base, swing1LimitDegrees=8, swing2LimitDegrees=18, twistLimitDegrees=0.01}
 {bone1Name=dyn_armVialA_Liquid, bone2Name=dyn_armVialA, swing1LimitDegrees=20, swing2LimitDegrees=20, twistLimitDegrees=20}
 {bone1Name=dyn_armVialB, bone2Name=dyn_armVialB_Base, swing1LimitDegrees=8, swing2LimitDegrees=18, twistLimitDegrees=0.01}
 {bone1Name=dyn_armVialB_Liquid, bone2Name=dyn_armVialB, swing1LimitDegrees=20, swing2LimitDegrees=20, twistLimitDegrees=20}
 {bone1Name=dyn_armVialC, bone2Name=dyn_armVialC_Base, swing1LimitDegrees=8, swing2LimitDegrees=18, twistLimitDegrees=0.01}
 {bone1Name=dyn_armVialC_Liquid, bone2Name=dyn_armVialC, swing1LimitDegrees=20, swing2LimitDegrees=20, twistLimitDegrees=20}
 {bone1Name=dyn_armVialD, bone2Name=dyn_armVialD_Base, swing1LimitDegrees=8, swing2LimitDegrees=18, twistLimitDegrees=0.01}
 {bone1Name=dyn_armVialD_Liquid, bone2Name=dyn_armVialD, swing1LimitDegrees=20, swing2LimitDegrees=20, twistLimitDegrees=20}
 {bone1Name=dyn_armVialE, bone2Name=dyn_armVialE_Base, swing1LimitDegrees=8, swing2LimitDegrees=18, twistLimitDegrees=0.01}
 {bone1Name=dyn_armVialE_Liquid, bone2Name=dyn_armVialE, swing1LimitDegrees=20, swing2LimitDegrees=20, twistLimitDegrees=20}
]
