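# Physics asset for the three leg vials on the M_MED_Freak male cosmetic body, covering their physical motion against the right thigh.
# Each vial forms a chain from a thigh twist capsule through a small base body to the vial and its liquid body.
# Vial-to-base joints lock twist and use tight swing angles, while base-to-thigh and liquid-to-vial joints leave all rotation free.
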PhysicsAsset /BRCosmetics/Characters/Player/Male/Medium/Bodies/M_MED_Freak/Meshes/M_MED_Freak_VialLeg_Physics
bodies {
 dyn_legVialA [{shapeType=Capsule, length=5.6764, radius=2.9939, center=<2.0271 0.0196 -0.0739>, rotation=<rot 0 89.8562 -89.0993>}]
 dyn_legVialA_Base [{shapeType=Capsule, length=0.505, radius=0.505, rotation=<rot 0 -90 0>}]
 dyn_legVialA_Liquid [{shapeType=Capsule, length=0.5572, radius=0.8355, center=<0 0 -0.4319>}]
 dyn_legVialB [{shapeType=Sphere, radius=3.703, center=<7.5453 -0.0002 0>}, {shapeType=Capsule, length=10.0236, radius=1.4026, rotation=<rot 90 0 0>}]
 dyn_legVialB_Base [{shapeType=Capsule, length=0.505, radius=0.505, rotation=<rot 0 -90 0>}]
 dyn_legVialB_Liquid [{shapeType=Capsule, length=0.5572, radius=0.8355, center=<0 0 -0.4319>}]
 dyn_legVialC [{shapeType=Capsule, length=13.0261, radius=1.5551, center=<4.8551 -0.0379 -0.0012>, rotation=<rot 0 -90.2474 -89.9677>}]
 dyn_legVialC_Base [{shapeType=Capsule, length=0.505, radius=0.505, rotation=<rot 0 -90 0>}]
 dyn_legVialC_Liquid [{shapeType=Capsule, length=0.5572, radius=0.8355, center=<0 0 -0.4319>}]
 thigh_twist_01_r [{shapeType=Capsule, length=17.5861, radius=12.2136, center=<-10.9916 -3.7408 0>, rotation=<rot 90 0 0>}]
}
constraints [
 {bone1Name=dyn_legVialA, bone2Name=dyn_legVialA_Base, swing1LimitDegrees=15, swing2LimitDegrees=30, twistLimitDegrees=45, twistMotion=Locked}
 {
  bone1Name=dyn_legVialA_Liquid
  bone2Name=dyn_legVialA
  swing1LimitDegrees=45
  swing1Motion=Free
  swing2LimitDegrees=45
  swing2Motion=Free
  twistLimitDegrees=45
  twistMotion=Free
 }
 {bone1Name=dyn_legVialB, bone2Name=dyn_legVialB_Base, swing1LimitDegrees=15, swing2LimitDegrees=30, twistLimitDegrees=45, twistMotion=Locked}
 {
  bone1Name=dyn_legVialB_Liquid
  bone2Name=dyn_legVialB
  swing1LimitDegrees=45
  swing1Motion=Free
  swing2LimitDegrees=45
  swing2Motion=Free
  twistLimitDegrees=45
  twistMotion=Free
 }
 {bone1Name=dyn_legVialC, bone2Name=dyn_legVialC_Base, swing1LimitDegrees=15, swing2LimitDegrees=30, twistLimitDegrees=45, twistMotion=Locked}
 {
  bone1Name=dyn_legVialC_Liquid
  bone2Name=dyn_legVialC
  swing1LimitDegrees=45
  swing1Motion=Free
  swing2LimitDegrees=45
  swing2Motion=Free
  twistLimitDegrees=45
  twistMotion=Free
 }
 {
  bone1Name=dyn_legVialA_Base
  bone2Name=thigh_twist_01_r
  swing1LimitDegrees=45
  swing1Motion=Free
  swing2LimitDegrees=45
  swing2Motion=Free
  twistLimitDegrees=45
  twistMotion=Free
 }
 {
  bone1Name=dyn_legVialB_Base
  bone2Name=thigh_twist_01_r
  swing1LimitDegrees=45
  swing1Motion=Free
  swing2LimitDegrees=45
  swing2Motion=Free
  twistLimitDegrees=45
  twistMotion=Free
 }
 {
  bone1Name=dyn_legVialC_Base
  bone2Name=thigh_twist_01_r
  swing1LimitDegrees=45
  swing1Motion=Free
  swing2LimitDegrees=45
  swing2Motion=Free
  twistLimitDegrees=45
  twistMotion=Free
 }
]
