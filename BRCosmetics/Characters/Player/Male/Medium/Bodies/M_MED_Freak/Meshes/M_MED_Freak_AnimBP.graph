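# Animation Blueprint for the Freak player body cosmetic, adding secondary motion to its vials and dangling props.
# It copies the attached parent's pose, constrains the chest-vial base, then applies a Control Rig fed by crouch, swing and jump values.
# Crouch-weighted bone offsets feed chained AnimDynamics nodes that swing arm and leg vials, with separate liquid-bone simulations for slosh.
# Alternate AnimDynamics chains and RigidBody nodes sit without pose inputs.
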AnimBlueprint /BRCosmetics/Characters/Player/Male/Medium/Bodies/M_MED_Freak/Meshes/M_MED_Freak_AnimBP
target_skeleton=&/M_MED_Freak_Skeleton
graphs [
 AnimGraph {nodes [
   Anim Dynamics - Bone: dyn_legVialA_Liquid(Alpha=1, ExternalForce="0,0,0") {
    angularDampingOverride=0.95
    angularSpringConstant=1000
    bAngularSpring=true
    bGravityOverrideInSimSpace=true
    bOverrideAngularDamping=true
    bUseGravityOverride=true
    simulationSpace=BoneRelative
    alphaBoolBlend {blendOption=Linear}
    boundBone {boneName=dyn_legVialA_Liquid}
    componentAppliedLinearAccClamp=<3000 3000 3000>
    componentLinearAccScale=<0.5 0.5 0.5>
    componentLinearVelScale=<0.5 0.5 0.5>
    gravityOverride {z=-980}
    physicsBodyDefinitions [{
      sphereCollisionRadius=0
      boundBone {boneName=dyn_legVialA_Liquid}
      boxExtents=<10 10 2>
      constraintSetup {
       angularTargetAxis=AxisZ
       twistAxis=AxisZ
       angularLimitsMax=<33 33 0>
       angularLimitsMin=<-33 -33 0>
       angularTarget=<0 0 0>
       linearAxesMax=<0 0 1>
      }
      localJointOffset=<0 0 -2>
     }]
   }
   n1 = Anim Dynamics - Bone: dyn_armVialA(Alpha=1, ExternalForce="0,0,0") {
    angularDampingOverride=0.999
    angularSpringConstant=800
    bAngularSpring=true
    bGravityOverrideInSimSpace=true
    bOverrideAngularDamping=true
    bUseGravityOverride=true
    simulationSpace=BoneRelative
    alphaBoolBlend {blendOption=Linear}
    boundBone {boneName=dyn_armVialA}
    componentAppliedLinearAccClamp=<800 800 800>
    componentLinearAccScale=<1 1 1>
    componentLinearVelScale=<1 1 1>
    gravityOverride {z=-980}
    physicsBodyDefinitions [{
      sphereCollisionRadius=0
      boundBone {boneName=dyn_armVialA}
      boxExtents=<8 22 22>
      constraintSetup {angularLimitsMax=<6 6 0>, angularLimitsMin=<-12 -12 0>, angularTarget=<0 0 0>}
      localJointOffset=<8 0 0>
     }]
    relativeSpaceBone {boneName=spine_03}
   }
   n2 = Anim Dynamics - Bone: dyn_armVialA_Liquid(ComponentPose=@n1, Alpha=1, ExternalForce="0,0,0") {
    angularDampingOverride=0.95
    angularSpringConstant=1000
    bAngularSpring=true
    bGravityOverrideInSimSpace=true
    bOverrideAngularDamping=true
    bUseGravityOverride=true
    linearSpringConstant=1000
    simulationSpace=BoneRelative
    alphaBoolBlend {blendOption=Linear}
    boundBone {boneName=dyn_armVialA_Liquid}
    componentAppliedLinearAccClamp=<3000 3000 3000>
    componentLinearAccScale=<0.5 0.5 0.5>
    componentLinearVelScale=<0.5 0.5 0.5>
    gravityOverride {z=-980}
    physicsBodyDefinitions [{
      sphereCollisionRadius=0
      boundBone {boneName=dyn_armVialA_Liquid}
      boxExtents=<10 10 2>
      constraintSetup {
       angularTargetAxis=AxisZ
       twistAxis=AxisZ
       angularLimitsMax=<28 28 0>
       angularLimitsMin=<-28 -28 0>
       angularTarget=<0 0 0>
       linearAxesMax=<0 0 0.8>
      }
      localJointOffset=<0 0 -2>
     }]
   }
   n3 = Anim Dynamics - Bone: dyn_armVialB(ComponentPose=@n2, Alpha=1, ExternalForce="0,0,0") {
    angularDampingOverride=0.999
    angularSpringConstant=800
    bAngularSpring=true
    bGravityOverrideInSimSpace=true
    bOverrideAngularDamping=true
    bUseGravityOverride=true
    simulationSpace=BoneRelative
    alphaBoolBlend {blendOption=Linear}
    boundBone {boneName=dyn_armVialB}
    componentAppliedLinearAccClamp=<800 800 800>
    componentLinearAccScale=<1 1 1>
    componentLinearVelScale=<1 1 1>
    gravityOverride {z=-980}
    physicsBodyDefinitions [{
      sphereCollisionRadius=0
      boundBone {boneName=dyn_armVialB}
      boxExtents=<8 22 22>
      constraintSetup {angularLimitsMax=<6 4 0>, angularLimitsMin=<-12 -8 0>, angularTarget=<0 0 0>}
      localJointOffset=<8 0 0>
     }]
    relativeSpaceBone {boneName=spine_03}
   }
   n4 = Anim Dynamics - Bone: dyn_armVialB_Liquid(ComponentPose=@n3, Alpha=1, ExternalForce="0,0,0") {
    angularDampingOverride=0.95
    angularSpringConstant=1000
    bAngularSpring=true
    bGravityOverrideInSimSpace=true
    bOverrideAngularDamping=true
    bUseGravityOverride=true
    linearSpringConstant=1000
    simulationSpace=BoneRelative
    alphaBoolBlend {blendOption=Linear}
    boundBone {boneName=dyn_armVialB_Liquid}
    componentAppliedLinearAccClamp=<3000 3000 3000>
    componentLinearAccScale=<0.5 0.5 0.5>
    componentLinearVelScale=<0.5 0.5 0.5>
    gravityOverride {z=-980}
    physicsBodyDefinitions [{
      sphereCollisionRadius=0
      boundBone {boneName=dyn_armVialB_Liquid}
      boxExtents=<10 10 2>
      constraintSetup {
       angularTargetAxis=AxisZ
       twistAxis=AxisZ
       angularLimitsMax=<33 33 0>
       angularLimitsMin=<-33 -33 0>
       angularTarget=<0 0 0>
       linearAxesMax=<0 0 1>
      }
      localJointOffset=<0 0 -2>
     }]
   }
   n5 = Anim Dynamics - Bone: dyn_armVialC(ComponentPose=@n4, Alpha=1, ExternalForce="0,0,0") {
    angularDampingOverride=0.999
    angularSpringConstant=800
    bAngularSpring=true
    bGravityOverrideInSimSpace=true
    bOverrideAngularDamping=true
    bUseGravityOverride=true
    simulationSpace=BoneRelative
    alphaBoolBlend {blendOption=Linear}
    boundBone {boneName=dyn_armVialC}
    componentAppliedLinearAccClamp=<800 800 800>
    componentLinearAccScale=<0.8 0.8 0.8>
    componentLinearVelScale=<0.8 0.8 0.8>
    gravityOverride {z=-980}
    physicsBodyDefinitions [{
      sphereCollisionRadius=0
      boundBone {boneName=dyn_armVialC}
      boxExtents=<4 16 16>
      constraintSetup {angularLimitsMax=<12 12 0>, angularLimitsMin=<-12 -6 0>, angularTarget=<0 0 0>}
      localJointOffset=<5 0 0>
     }]
    relativeSpaceBone {boneName=spine_03}
   }
   n6 = Anim Dynamics - Bone: dyn_armVialC_Liquid(ComponentPose=@n5, Alpha=1, ExternalForce="0,0,0") {
    angularDampingOverride=0.95
    angularSpringConstant=1000
    bAngularSpring=true
    bGravityOverrideInSimSpace=true
    bOverrideAngularDamping=true
    bUseGravityOverride=true
    linearSpringConstant=1000
    simulationSpace=BoneRelative
    alphaBoolBlend {blendOption=Linear}
    boundBone {boneName=dyn_armVialC_Liquid}
    componentAppliedLinearAccClamp=<3000 3000 3000>
    componentLinearAccScale=<0.5 0.5 0.5>
    componentLinearVelScale=<0.5 0.5 0.5>
    gravityOverride {z=-980}
    physicsBodyDefinitions [{
      sphereCollisionRadius=0
      boundBone {boneName=dyn_armVialC_Liquid}
      boxExtents=<10 10 2>
      constraintSetup {
       angularTargetAxis=AxisZ
       twistAxis=AxisZ
       angularLimitsMax=<15 15 0>
       angularLimitsMin=<-15 -15 0>
       angularTarget=<0 0 0>
       linearAxesMax=<0 0 0.5>
      }
      localJointOffset=<0 0 -2>
     }]
   }
   n7 = Anim Dynamics - Bone: dyn_armVialD(ComponentPose=@n6, Alpha=1, ExternalForce="0,0,0") {
    angularDampingOverride=0.999
    angularSpringConstant=800
    bAngularSpring=true
    bGravityOverrideInSimSpace=true
    bOverrideAngularDamping=true
    bUseGravityOverride=true
    simulationSpace=BoneRelative
    alphaBoolBlend {blendOption=Linear}
    boundBone {boneName=dyn_armVialD}
    componentAppliedLinearAccClamp=<400 400 400>
    componentLinearAccScale=<1 1 1>
    componentLinearVelScale=<1 1 1>
    gravityOverride {z=-980}
    physicsBodyDefinitions [{
      sphereCollisionRadius=0
      boundBone {boneName=dyn_armVialD}
      boxExtents=<3 12 12>
      constraintSetup {angularLimitsMax=<12 12 0>, angularLimitsMin=<-12 -12 0>, angularTarget=<0 0 0>}
      localJointOffset=<4 0 0>
     }]
    relativeSpaceBone {boneName=spine_03}
   }
   n8 = Anim Dynamics - Bone: dyn_armVialD_Liquid(ComponentPose=@n7, Alpha=1, ExternalForce="0,0,0") {
    angularDampingOverride=0.95
    angularSpringConstant=1000
    bAngularSpring=true
    bGravityOverrideInSimSpace=true
    bOverrideAngularDamping=true
    bUseGravityOverride=true
    linearSpringConstant=1000
    simulationSpace=BoneRelative
    alphaBoolBlend {blendOption=Linear}
    boundBone {boneName=dyn_armVialD_Liquid}
    componentAppliedLinearAccClamp=<3000 3000 3000>
    componentLinearAccScale=<0.5 0.5 0.5>
    componentLinearVelScale=<0.5 0.5 0.5>
    gravityOverride {z=-980}
    physicsBodyDefinitions [{
      sphereCollisionRadius=0
      boundBone {boneName=dyn_armVialD_Liquid}
      boxExtents=<10 10 2>
      constraintSetup {
       angularTargetAxis=AxisZ
       twistAxis=AxisZ
       angularLimitsMax=<33 33 0>
       angularLimitsMin=<-33 -33 0>
       angularTarget=<0 0 0>
       linearAxesMax=<0 0 1>
      }
      localJointOffset=<0 0 -2>
     }]
   }
   n9 = Anim Dynamics - Bone: dyn_armVialE(ComponentPose=@n8, Alpha=1, ExternalForce="0,0,0") {
    angularDampingOverride=0.999
    angularSpringConstant=800
    bAngularSpring=true
    bGravityOverrideInSimSpace=true
    bOverrideAngularDamping=true
    bUseGravityOverride=true
    bUsePlanarLimit=false
    simulationSpace=BoneRelative
    alphaBoolBlend {blendOption=Linear}
    boundBone {boneName=dyn_armVialE}
    componentAppliedLinearAccClamp=<400 400 400>
    componentLinearAccScale=<1 1 1>
    componentLinearVelScale=<1 1 1>
    gravityOverride {z=-980}
    physicsBodyDefinitions [{
      sphereCollisionRadius=0
      boundBone {boneName=dyn_armVialE}
      boxExtents=<4 11 11>
      constraintSetup {angularLimitsMax=<12 10 0>, angularLimitsMin=<-14 -10 0>, angularTarget=<0 0 0>}
      localJointOffset=<6 0 0>
     }]
    relativeSpaceBone {boneName=spine_03}
   }
   Anim Dynamics - Bone: dyn_armVialE_Liquid(ComponentPose=@n9, Alpha=1, ExternalForce="0,0,0") {
    angularDampingOverride=0.95
    angularSpringConstant=1000
    bAngularSpring=true
    bGravityOverrideInSimSpace=true
    bOverrideAngularDamping=true
    bUseGravityOverride=true
    linearSpringConstant=1000
    simulationSpace=BoneRelative
    alphaBoolBlend {blendOption=Linear}
    boundBone {boneName=dyn_armVialE_Liquid}
    componentAppliedLinearAccClamp=<3000 3000 3000>
    componentLinearAccScale=<0.5 0.5 0.5>
    componentLinearVelScale=<0.5 0.5 0.5>
    gravityOverride {z=-980}
    physicsBodyDefinitions [{
      sphereCollisionRadius=0
      boundBone {boneName=dyn_armVialE_Liquid}
      boxExtents=<10 10 2>
      constraintSetup {
       angularTargetAxis=AxisZ
       twistAxis=AxisZ
       angularLimitsMax=<28 28 0>
       angularLimitsMin=<-28 -28 0>
       angularTarget=<0 0 0>
       linearAxesMax=<0 0 1>
      }
      localJointOffset=<0 0 -2>
     }]
   }
   n10 = Anim Dynamics - Bone: jug(Alpha=1, ExternalForce="0,0,0") {
    angularDampingOverride=0.999
    angularSpringConstant=500
    bAngularSpring=true
    bGravityOverrideInSimSpace=true
    bLinearSpring=true
    bOverrideAngularDamping=true
    bOverrideLinearDamping=true
    gravityScale=0.6
    linearDampingOverride=0.999
    linearSpringConstant=50
    alphaBoolBlend {blendOption=Linear}
    boundBone {boneName=jug}
    componentAppliedLinearAccClamp=<1000 1000 1000>
    componentLinearAccScale=<1 1 1>
    componentLinearVelScale=<1 1 1>
    physicsBodyDefinitions [{
      sphereCollisionRadius=0
      boundBone {boneName=jug}
      boxExtents=<45 45 22>
      constraintSetup {angularLimitsMax=<15 2 0>, angularLimitsMin=<-15 -2 -44>, angularTarget=<0 0 0>}
      localJointOffset=<2 0 -26>
     }]
   }
   n11 = Anim Dynamics - Bone: dyn_jugJuice(ComponentPose=@n10, Alpha=1, ExternalForce="0,0,0") {
    angularDampingOverride=0.999
    angularSpringConstant=1000
    bAngularSpring=true
    bGravityOverrideInSimSpace=true
    bOverrideAngularDamping=true
    alphaBoolBlend {blendOption=Linear}
    boundBone {boneName=dyn_jugJuice}
    componentAppliedLinearAccClamp=<50 50 50>
    componentLinearAccScale=<1 1 1>
    componentLinearVelScale=<1 1 1>
    physicsBodyDefinitions [{
      sphereCollisionRadius=0
      boundBone {boneName=dyn_jugJuice}
      boxExtents=<18 22 8>
      constraintSetup {angularLimitsMax=<30 30 30>, angularLimitsMin=<-30 -30 -30>, angularTarget=<0 0 0>}
      localJointOffset=<0 -0.05 -11>
     }]
   }
   n12 = Anim Dynamics - Bone: dyn_jugstopper(ComponentPose=@n11, Alpha=1, ExternalForce="0,0,0") {
    angularDampingOverride=0.999
    angularSpringConstant=500
    bAngularSpring=true
    bGravityOverrideInSimSpace=true
    bLinearSpring=true
    bOverrideAngularDamping=true
    gravityScale=0.1
    linearSpringConstant=500
    alphaBoolBlend {blendOption=Linear}
    boundBone {boneName=dyn_jugstopper}
    componentAppliedLinearAccClamp=<3000 3000 3000>
    componentLinearAccScale=<1 1 1>
    componentLinearVelScale=<1 1 1>
    physicsBodyDefinitions [{
      sphereCollisionRadius=0
      boundBone {boneName=dyn_jugstopper}
      boxExtents=<2 2 2>
      constraintSetup {angularLimitsMax=<2 2 1>, angularLimitsMin=<-2 -2 -1>, angularTarget=<0 0 0>}
      localJointOffset=<-0.1 0 -0.1>
     }]
   }
   n13 = Anim Dynamics - Bone: dyn_vial_l(ComponentPose=@n12, Alpha=1, ExternalForce="0,0,0") {
    angularDampingOverride=0.999
    angularSpringConstant=1000
    bAngularSpring=true
    bGravityOverrideInSimSpace=true
    bOverrideAngularDamping=true
    linearSpringConstant=1000
    alphaBoolBlend {blendOption=Linear}
    boundBone {boneName=dyn_vial_l}
    componentAppliedLinearAccClamp=<1000 1000 1000>
    componentLinearAccScale=<1 1 1>
    componentLinearVelScale=<1 1 1>
    physicsBodyDefinitions [{
      sphereCollisionRadius=0
      boundBone {boneName=dyn_vial_l}
      boxExtents=<14 16 4>
      constraintSetup {angularTargetAxis=AxisY, twistAxis=AxisY, angularLimitsMax=<12 12 0>, angularLimitsMin=<-12 -12 0>, angularTarget=<0 0 0>}
      localJointOffset=<0 -0.05 -8>
     }]
   }
   n14 = Anim Dynamics - Bone: dyn_vialJucie_l(ComponentPose=@n13, Alpha=1, ExternalForce="0,0,0") {
    angularDampingOverride=0.95
    angularSpringConstant=1000
    bGravityOverrideInSimSpace=true
    bOverrideAngularDamping=true
    alphaBoolBlend {blendOption=Linear}
    boundBone {boneName=dyn_vialJucie_l}
    componentAppliedLinearAccClamp=<3000 3000 3000>
    componentLinearAccScale=<1 1 1>
    componentLinearVelScale=<1 1 1>
    physicsBodyDefinitions [{
      sphereCollisionRadius=0
      boundBone {boneName=dyn_vialJucie_l}
      boxExtents=<2 2 2>
      constraintSetup {
       angularTargetAxis=AxisY
       twistAxis=AxisY
       angularLimitsMax=<33 33 0>
       angularLimitsMin=<-33 -33 0>
       angularTarget=<0 0 0>
       linearAxesMax=<0 0 1>
      }
      localJointOffset=<0 0 -1>
     }]
   }
   n15 = Anim Dynamics - Bone: dyn_vial_c(ComponentPose=@n14, Alpha=1, ExternalForce="0,0,0") {
    angularDampingOverride=0.999
    angularSpringConstant=1000
    bAngularSpring=true
    bGravityOverrideInSimSpace=true
    bOverrideAngularDamping=true
    alphaBoolBlend {blendOption=Linear}
    boundBone {boneName=dyn_vial_c}
    componentAppliedLinearAccClamp=<3000 3000 3000>
    componentLinearAccScale=<1 1 1>
    componentLinearVelScale=<1 1 1>
    physicsBodyDefinitions [{
      sphereCollisionRadius=0
      boundBone {boneName=dyn_vial_c}
      boxExtents=<14 30 3>
      constraintSetup {angularTargetAxis=AxisY, twistAxis=AxisY, angularLimitsMax=<12 12 0>, angularLimitsMin=<-12 -12 0>, angularTarget=<0 0 0>}
      localJointOffset=<0 -0.05 -7>
     }]
   }
   n16 = Anim Dynamics - Bone: dyn_vialJucie_c(ComponentPose=@n15, Alpha=1, ExternalForce="0,0,0") {
    angularDampingOverride=0.95
    angularSpringConstant=1000
    bGravityOverrideInSimSpace=true
    bOverrideAngularDamping=true
    alphaBoolBlend {blendOption=Linear}
    boundBone {boneName=dyn_vialJucie_c}
    componentAppliedLinearAccClamp=<3000 3000 3000>
    componentLinearAccScale=<1 1 1>
    componentLinearVelScale=<1 1 1>
    physicsBodyDefinitions [{
      sphereCollisionRadius=0
      boundBone {boneName=dyn_vialJucie_c}
      boxExtents=<2 2 2>
      constraintSetup {
       angularTargetAxis=AxisY
       twistAxis=AxisY
       angularLimitsMax=<33 33 0>
       angularLimitsMin=<-33 -33 0>
       angularTarget=<0 0 0>
       linearAxesMax=<0 0 1>
      }
      localJointOffset=<0 0 -1>
     }]
   }
   n17 = Anim Dynamics - Bone: dyn_vial_r(ComponentPose=@n16, Alpha=1, ExternalForce="0,0,0") {
    angularDampingOverride=0.999
    angularSpringConstant=1000
    bAngularSpring=true
    bGravityOverrideInSimSpace=true
    bOverrideAngularDamping=true
    alphaBoolBlend {blendOption=Linear}
    boundBone {boneName=dyn_vial_r}
    componentAppliedLinearAccClamp=<3000 3000 3000>
    componentLinearAccScale=<1 1 1>
    componentLinearVelScale=<1 1 1>
    physicsBodyDefinitions [{
      sphereCollisionRadius=0
      boundBone {boneName=dyn_vial_r}
      boxExtents=<18 22 8>
      constraintSetup {angularTargetAxis=AxisY, twistAxis=AxisY, angularLimitsMax=<12 12 0>, angularLimitsMin=<-12 -12 0>, angularTarget=<0 0 0>}
      localJointOffset=<0 -0.05 -11>
     }]
   }
   n18 = Anim Dynamics - Bone: dyn_vialJucie_r(ComponentPose=@n17, Alpha=1, ExternalForce="0,0,0") {
    angularDampingOverride=0.95
    angularSpringConstant=1000
    bGravityOverrideInSimSpace=true
    bOverrideAngularDamping=true
    alphaBoolBlend {blendOption=Linear}
    boundBone {boneName=dyn_vialJucie_r}
    componentAppliedLinearAccClamp=<3000 3000 3000>
    componentLinearAccScale=<1 1 1>
    componentLinearVelScale=<1 1 1>
    physicsBodyDefinitions [{
      sphereCollisionRadius=0
      boundBone {boneName=dyn_vialJucie_r}
      boxExtents=<2 2 2>
      constraintSetup {
       angularTargetAxis=AxisY
       twistAxis=AxisY
       angularLimitsMax=<33 33 0>
       angularLimitsMin=<-33 -33 0>
       angularTarget=<0 0 0>
       linearAxesMax=<0 0 1>
      }
      localJointOffset=<0 0 -1>
     }]
   }
   n19 = Anim Dynamics - Bone: dyn_bellslingpivot(ComponentPose=@n18, Alpha=1, ExternalForce="0,0,0") {
    angularDampingOverride=0.999
    angularSpringConstant=1500
    bAngularSpring=true
    bGravityOverrideInSimSpace=true
    bOverrideAngularDamping=true
    bOverrideLinearDamping=true
    linearDampingOverride=0.999
    alphaBoolBlend {blendOption=Linear}
    boundBone {boneName=dyn_bellslingpivot}
    componentAppliedLinearAccClamp=<1000 1000 1000>
    componentLinearAccScale=<1 1 1>
    componentLinearVelScale=<1 1 1>
    physicsBodyDefinitions [{
      sphereCollisionRadius=0
      boundBone {boneName=dyn_bellslingpivot}
      boxExtents=<8 22 26>
      constraintSetup {angularLimitsMax=<0 12 12>, angularLimitsMin=<0 -12 -12>, angularTarget=<0 0 0>, linearAxesMin=<-6 0 0>}
      localJointOffset=<20 0 -1>
     }]
   }
   n20 = "Transform (Modify) Bone - Bone: dyn_bellslingpivot"!(ComponentPose=@n19, Alpha=1) -> (Pose) {
    rotationMode=BMM_Additive
    translationMode=BMM_Additive
    alphaBoolBlend {blendOption=Linear}
    boneToModify {boneName=dyn_bellslingpivot}
    rotation {roll=-33}
    translation {y=-0.5, z=-1}
   }
   n21 = Anim Dynamics - Bone: dyn_bellclapper(ComponentPose=@n20, Alpha=0.5, ExternalForce="0,0,0") {
    angularDampingOverride=0.999
    angularSpringConstant=1000
    bAngularSpring=true
    bGravityOverrideInSimSpace=true
    bOverrideAngularDamping=true
    alphaBoolBlend {blendOption=Linear}
    boundBone {boneName=dyn_bellclapper}
    componentAppliedLinearAccClamp=<400 400 400>
    componentLinearAccScale=<1 1 1>
    componentLinearVelScale=<1 1 1>
    physicsBodyDefinitions [{
      sphereCollisionRadius=0
      boundBone {boneName=dyn_bellclapper}
      boxExtents=<8 22 26>
      constraintSetup {angularLimitsMax=<0 1 1>, angularLimitsMin=<0 -1 -1>, angularTarget=<0 0 0>}
      localJointOffset=<20 0 -1>
     }]
   }
   n22 = Anim Dynamics - Bone: dyn_syringestopper(ComponentPose=@n21, Alpha=1, ExternalForce="0,0,0") {
    angularDampingOverride=0.999
    angularSpringConstant=2000
    bAngularSpring=true
    bGravityOverrideInSimSpace=true
    bOverrideAngularDamping=true
    gravityScale=0.01
    alphaBoolBlend {blendOption=Linear}
    boundBone {boneName=dyn_syringestopper}
    componentAppliedLinearAccClamp=<200 200 200>
    componentLinearAccScale=<1 1 1>
    componentLinearVelScale=<1 1 1>
    physicsBodyDefinitions [{
      sphereCollisionRadius=0
      boundBone {boneName=dyn_syringestopper}
      boxExtents=<2 12 12>
      constraintSetup {angularTargetAxis=AxisY, twistAxis=AxisZ, angularLimitsMax=<1 3 3>, angularLimitsMin=<-1 -3 -3>, angularTarget=<0 0 0>}
      localJointOffset=<6 0 -4>
     }]
   }
   Anim Dynamics - Bone: dyn_syringeJuice(ComponentPose=@n22, Alpha=1, ExternalForce="0,0,0") {
    angularDampingOverride=0.95
    angularSpringConstant=2000
    bGravityOverrideInSimSpace=true
    bOverrideAngularDamping=true
    gravityScale=0.01
    alphaBoolBlend {blendOption=Linear}
    boundBone {boneName=dyn_syringeJuice}
    componentAppliedLinearAccClamp=<100 100 100>
    componentLinearAccScale=<1 1 1>
    componentLinearVelScale=<1 1 1>
    physicsBodyDefinitions [{
      sphereCollisionRadius=0
      boundBone {boneName=dyn_syringeJuice}
      boxExtents=<1 1 1>
      constraintSetup {angularTargetAxis=AxisY, angularLimitsMax=<33 33 0>, angularLimitsMin=<-33 -33 0>, angularTarget=<0 0 0>}
      localJointOffset=<0 0 -1>
     }]
   }
   Anim Dynamics - Bone: dyn_legVialA_Liquid(Alpha=1, ExternalForce="0,0,0") {
    angularDampingOverride=0.95
    angularSpringConstant=1000
    bAngularSpring=true
    bGravityOverrideInSimSpace=true
    bOverrideAngularDamping=true
    bUseGravityOverride=true
    simulationSpace=BoneRelative
    alphaBoolBlend {blendOption=Linear}
    boundBone {boneName=dyn_legVialA_Liquid}
    componentAppliedLinearAccClamp=<3000 3000 3000>
    componentLinearAccScale=<0.5 0.5 0.5>
    componentLinearVelScale=<0.5 0.5 0.5>
    gravityOverride {z=-980}
    physicsBodyDefinitions [{
      sphereCollisionRadius=0
      boundBone {boneName=dyn_legVialA_Liquid}
      boxExtents=<10 10 2>
      constraintSetup {
       angularTargetAxis=AxisZ
       twistAxis=AxisZ
       angularLimitsMax=<33 33 0>
       angularLimitsMin=<-33 -33 0>
       angularTarget=<0 0 0>
       linearAxesMax=<0 0 1>
      }
      localJointOffset=<0 0 -2>
     }]
   }
   RigidBody() {
    bClampLinearTranslationLimitToRefPose=true
    bOverrideWorldGravity=true
    overridePhysicsAsset=&/M_MED_Freak_VialLeg_Physics
    simulationSpace=BaseBoneSpace
    alphaBoolBlend {blendOption=Linear}
    baseBoneRef {boneName=pelvis}
    componentAppliedLinearAccClamp=<4000 4000 4000>
    componentLinearAccScale=<0.4 0.4 0.5>
    componentLinearVelScale=<0.2 0.2 0.3>
    overrideWorldGravity {z=-980}
   }
   RigidBody() {
    bClampLinearTranslationLimitToRefPose=true
    bOverrideWorldGravity=true
    overridePhysicsAsset=&/M_MED_Freak_VialArms_Physics
    simulationSpace=BaseBoneSpace
    alphaBoolBlend {blendOption=Linear}
    baseBoneRef {boneName=spine_05}
    componentAppliedLinearAccClamp=<4000 4000 4000>
    componentLinearAccScale=<0.4 0.4 0.5>
    componentLinearVelScale=<0.2 0.2 0.3>
    overrideWorldGravity {z=-980}
   }
   RigidBody() {
    bClampLinearTranslationLimitToRefPose=true
    bOverrideWorldGravity=true
    overridePhysicsAsset=&/M_MED_Freak_Physics
    simulationSpace=BaseBoneSpace
    alphaBoolBlend {blendOption=Linear}
    baseBoneRef {boneName=spine_03}
    componentAppliedLinearAccClamp=<4000 4000 4000>
    componentLinearAccScale=<0.4 0.4 0.5>
    componentLinearVelScale=<0.2 0.2 0.3>
    overrideWorldGravity {z=-980}
   }
   RigidBody() {
    bClampLinearTranslationLimitToRefPose=true
    bOverrideWorldGravity=true
    overridePhysicsAsset=&M_MED_Freak_Physics
    simulationSpace=BaseBoneSpace
    alphaBoolBlend {blendOption=Linear}
    baseBoneRef {boneName=root}
    componentAppliedLinearAccClamp=<4000 4000 4000>
    componentLinearAccScale=<0.4 0.4 0.5>
    componentLinearVelScale=<0.2 0.2 0.3>
    overrideWorldGravity {z=-980}
   }
   n23 = "Constraint - dyn_chestVial_Base (PPPPPP)"!(
    ComponentPose=Local To Component(LocalPose=Copy Pose From Mesh() {bUseAttachedParent=true})
    Alpha=1
    ConstraintWeights_0=0.1248
    ConstraintWeights_1:float=0.095
    ConstraintWeights_2:float=0.0704
    ConstraintWeights_3:float=0.006
    ConstraintWeights_4:float=0.7003
    ConstraintWeights_5:float=0.0035
   ) -> (Pose) {
    alphaBoolBlend {blendOption=Linear}
    boneToModify {boneName=dyn_chestVial_Base}
    constraintSetup [
     {targetBone {boneName=spine_01}}
     {targetBone {boneName=spine_02}}
     {targetBone {boneName=spine_03}}
     {targetBone {boneName=spine_04}}
     {targetBone {boneName=pelvis}}
     {targetBone {boneName=thigh_twist_01_l}}
    ]
    constraintWeights [1,1,1,1,1,1]
   }
   n24 = Control Rig!(
    Source=Component To Local(ComponentPose=@n23)
    Alpha
    crouching:double=Get crouchAlpha() -> (crouchAlpha:double)
    targeting:double=Get SwingDelta() -> (SwingDelta:double)
    jumping:double=Get jumpAlpha() -> (jumpAlpha:double)
   ) -> (Pose) {
    alphaBoolBlend {blendOption=Linear}
    controlRigAssetReference {blueprintRigClass=&/M_MED_Freak_CtrlRig.M_MED_Freak_CtrlRig_C}
   }
   n25 = Get crouchAlpha()
   n26 = "Transform (Modify) Bone - Bone: dyn_chestVial_Base"(ComponentPose=Local To Component(LocalPose=@n24), Alpha=@n25, Translation="-6,9.092,-14.195", Rotation="-33.180,16.440,-75", Scale="1,1,1") {
    rotationMode=BMM_Additive
    rotationSpace=BCS_BoneSpace
    translationMode=BMM_Additive
    translationSpace=BCS_BoneSpace
    alphaBoolBlend {blendOption=Linear}
    alphaScaleBiasClamp {bInterpResult=true, interpSpeedDecreasing=8, interpSpeedIncreasing=8}
    boneToModify {boneName=dyn_chestVial_Base}
    rotation=<rot -33.18 16.44 -75>
    translation=<-6 9.092 -14.195>
   }
   n27 = "Transform (Modify) Bone - Bone: dyn_frogMain_Base"(ComponentPose=@n26, Alpha=@n25, Translation="-12.987,-4.080,-12.547", Rotation="-65,-57,60", Scale="1,1,1") {
    rotationMode=BMM_Additive
    rotationSpace=BCS_BoneSpace
    translationMode=BMM_Additive
    translationSpace=BCS_BoneSpace
    alphaBoolBlend {blendOption=Linear}
    alphaScaleBiasClamp {bInterpResult=true, interpSpeedDecreasing=8, interpSpeedIncreasing=8}
    boneToModify {boneName=dyn_frogMain_Base}
    rotation=<rot -65 -57 60>
    translation=<-12.987 -4.08 -12.547>
   }
   n28 = Anim Dynamics - Bone: dyn_armVialA(ComponentPose=@n27, Alpha=1, ExternalForce="0,0,0") {
    angularDampingOverride=0.999
    angularSpringConstant=20000
    bGravityOverrideInSimSpace=true
    bOverrideAngularDamping=true
    bUseGravityOverride=true
    linearSpringConstant=150
    simulationSpace=RootRelative
    alphaBoolBlend {blendOption=Linear}
    boundBone {boneName=dyn_armVialA}
    componentAppliedLinearAccClamp=<3000 3000 3000>
    componentLinearAccScale=<0.5 0.5 0.5>
    componentLinearVelScale=<0.5 0.5 0.5>
    gravityOverride {z=-980}
    physicsBodyDefinitions [{
      sphereCollisionRadius=0
      boundBone {boneName=dyn_armVialA}
      boxExtents=<4 22 22>
      constraintSetup {angularLimitsMax=<4 9.5 0.1>, angularLimitsMin=<-4 -9.5 0.1>, angularTarget=<0 1 0>}
      localJointOffset=<6 0 0>
     }]
   }
   n29 = Anim Dynamics - Bone: dyn_armVialA_Liquid(ComponentPose=@n28, Alpha=1, ExternalForce="0,0,0") {
    angularDampingOverride=0.95
    angularSpringConstant=1000
    bAngularSpring=true
    bGravityOverrideInSimSpace=true
    bOverrideAngularDamping=true
    bUseGravityOverride=true
    linearSpringConstant=1000
    simulationSpace=BoneRelative
    alphaBoolBlend {blendOption=Linear}
    boundBone {boneName=dyn_armVialA_Liquid}
    componentAppliedLinearAccClamp=<3000 3000 3000>
    componentLinearAccScale=<0.5 0.5 0.5>
    componentLinearVelScale=<0.5 0.5 0.5>
    gravityOverride {z=-980}
    physicsBodyDefinitions [{
      sphereCollisionRadius=0
      boundBone {boneName=dyn_armVialA_Liquid}
      boxExtents=<10 10 2>
      constraintSetup {
       angularTargetAxis=AxisZ
       twistAxis=AxisZ
       angularLimitsMax=<28 28 0>
       angularLimitsMin=<-28 -28 0>
       angularTarget=<0 0 0>
       linearAxesMax=<0 0 0.8>
      }
      localJointOffset=<0 0 -2>
     }]
   }
   n30 = Anim Dynamics - Bone: dyn_armVialB(ComponentPose=@n29, Alpha=1, ExternalForce="0,0,0") {
    angularDampingOverride=0.999
    angularSpringConstant=20000
    bGravityOverrideInSimSpace=true
    bOverrideAngularDamping=true
    bUseGravityOverride=true
    linearSpringConstant=150
    simulationSpace=RootRelative
    alphaBoolBlend {blendOption=Linear}
    boundBone {boneName=dyn_armVialB}
    componentAppliedLinearAccClamp=<3000 3000 3000>
    componentLinearAccScale=<0.5 0.5 0.5>
    componentLinearVelScale=<0.5 0.5 0.5>
    gravityOverride {z=-980}
    physicsBodyDefinitions [{
      sphereCollisionRadius=0
      boundBone {boneName=dyn_armVialB}
      boxExtents=<4 18 18>
      constraintSetup {angularLimitsMax=<4 9 0.1>, angularLimitsMin=<-4 -9 0.1>, angularTarget=<0 1 0>}
      localJointOffset=<6 0 0>
     }]
   }
   n31 = Anim Dynamics - Bone: dyn_armVialB_Liquid(ComponentPose=@n30, Alpha=1, ExternalForce="0,0,0") {
    angularDampingOverride=0.95
    angularSpringConstant=1000
    bAngularSpring=true
    bGravityOverrideInSimSpace=true
    bOverrideAngularDamping=true
    bUseGravityOverride=true
    linearSpringConstant=1000
    simulationSpace=BoneRelative
    alphaBoolBlend {blendOption=Linear}
    boundBone {boneName=dyn_armVialB_Liquid}
    componentAppliedLinearAccClamp=<3000 3000 3000>
    componentLinearAccScale=<0.5 0.5 0.5>
    componentLinearVelScale=<0.5 0.5 0.5>
    gravityOverride {z=-980}
    physicsBodyDefinitions [{
      sphereCollisionRadius=0
      boundBone {boneName=dyn_armVialB_Liquid}
      boxExtents=<10 10 2>
      constraintSetup {
       angularTargetAxis=AxisZ
       twistAxis=AxisZ
       angularLimitsMax=<33 33 0>
       angularLimitsMin=<-33 -33 0>
       angularTarget=<0 0 0>
       linearAxesMax=<0 0 1>
      }
      localJointOffset=<0 0 -2>
     }]
   }
   n32 = Anim Dynamics - Bone: dyn_armVialC(ComponentPose=@n31, Alpha=1, ExternalForce="0,0,0") {
    angularDampingOverride=0.999
    angularSpringConstant=20000
    bGravityOverrideInSimSpace=true
    bOverrideAngularDamping=true
    bUseGravityOverride=true
    linearSpringConstant=150
    simulationSpace=RootRelative
    alphaBoolBlend {blendOption=Linear}
    boundBone {boneName=dyn_armVialC}
    componentAppliedLinearAccClamp=<3000 3000 3000>
    componentLinearAccScale=<0.5 0.5 0.5>
    componentLinearVelScale=<0.5 0.5 0.5>
    gravityOverride {z=-980}
    physicsBodyDefinitions [{
      sphereCollisionRadius=0
      boundBone {boneName=dyn_armVialC}
      boxExtents=<4 18 18>
      constraintSetup {angularLimitsMax=<2 8.5 0.1>, angularLimitsMin=<-2 -8.5 0.1>, angularTarget=<0 1 0>}
      localJointOffset=<6 0 0>
     }]
   }
   n33 = Anim Dynamics - Bone: dyn_armVialC_Liquid(ComponentPose=@n32, Alpha=1, ExternalForce="0,0,0") {
    angularDampingOverride=0.95
    angularSpringConstant=1000
    bAngularSpring=true
    bGravityOverrideInSimSpace=true
    bOverrideAngularDamping=true
    bUseGravityOverride=true
    linearSpringConstant=1000
    simulationSpace=BoneRelative
    alphaBoolBlend {blendOption=Linear}
    boundBone {boneName=dyn_armVialC_Liquid}
    componentAppliedLinearAccClamp=<3000 3000 3000>
    componentLinearAccScale=<0.5 0.5 0.5>
    componentLinearVelScale=<0.5 0.5 0.5>
    gravityOverride {z=-980}
    physicsBodyDefinitions [{
      sphereCollisionRadius=0
      boundBone {boneName=dyn_armVialC_Liquid}
      boxExtents=<10 10 2>
      constraintSetup {
       angularTargetAxis=AxisZ
       twistAxis=AxisZ
       angularLimitsMax=<15 15 0>
       angularLimitsMin=<-15 -15 0>
       angularTarget=<0 0 0>
       linearAxesMax=<0 0 0.5>
      }
      localJointOffset=<0 0 -2>
     }]
   }
   n34 = Anim Dynamics - Bone: dyn_armVialD(ComponentPose=@n33, Alpha=1, ExternalForce="0,0,0") {
    angularDampingOverride=0.999
    angularSpringConstant=20000
    bGravityOverrideInSimSpace=true
    bOverrideAngularDamping=true
    bUseGravityOverride=true
    linearSpringConstant=150
    simulationSpace=RootRelative
    alphaBoolBlend {blendOption=Linear}
    boundBone {boneName=dyn_armVialD}
    componentAppliedLinearAccClamp=<3000 3000 3000>
    componentLinearAccScale=<0.5 0.5 0.5>
    componentLinearVelScale=<0.5 0.5 0.5>
    gravityOverride {z=-980}
    physicsBodyDefinitions [{
      sphereCollisionRadius=0
      boundBone {boneName=dyn_armVialD}
      boxExtents=<4 15 15>
      constraintSetup {angularLimitsMax=<4 8.5 0.1>, angularLimitsMin=<-4 -8 -0.1>, angularTarget=<0 1 0>}
      localJointOffset=<6 0 0>
     }]
   }
   n35 = Anim Dynamics - Bone: dyn_armVialD_Liquid(ComponentPose=@n34, Alpha=1, ExternalForce="0,0,0") {
    angularDampingOverride=0.95
    angularSpringConstant=1000
    bAngularSpring=true
    bGravityOverrideInSimSpace=true
    bOverrideAngularDamping=true
    bUseGravityOverride=true
    linearSpringConstant=1000
    simulationSpace=BoneRelative
    alphaBoolBlend {blendOption=Linear}
    boundBone {boneName=dyn_armVialD_Liquid}
    componentAppliedLinearAccClamp=<3000 3000 3000>
    componentLinearAccScale=<0.5 0.5 0.5>
    componentLinearVelScale=<0.5 0.5 0.5>
    gravityOverride {z=-980}
    physicsBodyDefinitions [{
      sphereCollisionRadius=0
      boundBone {boneName=dyn_armVialD_Liquid}
      boxExtents=<10 10 2>
      constraintSetup {
       angularTargetAxis=AxisZ
       twistAxis=AxisZ
       angularLimitsMax=<33 33 0>
       angularLimitsMin=<-33 -33 0>
       angularTarget=<0 0 0>
       linearAxesMax=<0 0 1>
      }
      localJointOffset=<0 0 -2>
     }]
   }
   n36 = Anim Dynamics - Bone: dyn_armVialE(ComponentPose=@n35, Alpha=1, ExternalForce="0,0,0") {
    angularDampingOverride=0.999
    angularSpringConstant=20000
    bGravityOverrideInSimSpace=true
    bOverrideAngularDamping=true
    bUseGravityOverride=true
    linearSpringConstant=150
    simulationSpace=RootRelative
    alphaBoolBlend {blendOption=Linear}
    boundBone {boneName=dyn_armVialE}
    componentAppliedLinearAccClamp=<3000 3000 3000>
    componentLinearAccScale=<0.5 0.5 0.5>
    componentLinearVelScale=<0.5 0.5 0.5>
    gravityOverride {z=-980}
    physicsBodyDefinitions [{
      sphereCollisionRadius=0
      boundBone {boneName=dyn_armVialE}
      boxExtents=<4 15 15>
      constraintSetup {angularLimitsMax=<4 7 0.1>, angularLimitsMin=<-4 -6.5 -0.1>, angularTarget=<0 1 0>}
      localJointOffset=<6 0 0>
     }]
   }
   n37 = Anim Dynamics - Bone: dyn_armVialE_Liquid(ComponentPose=@n36, Alpha=1, ExternalForce="0,0,0") {
    angularDampingOverride=0.95
    angularSpringConstant=1000
    bAngularSpring=true
    bGravityOverrideInSimSpace=true
    bOverrideAngularDamping=true
    bUseGravityOverride=true
    linearSpringConstant=1000
    simulationSpace=BoneRelative
    alphaBoolBlend {blendOption=Linear}
    boundBone {boneName=dyn_armVialE_Liquid}
    componentAppliedLinearAccClamp=<3000 3000 3000>
    componentLinearAccScale=<0.5 0.5 0.5>
    componentLinearVelScale=<0.5 0.5 0.5>
    gravityOverride {z=-980}
    physicsBodyDefinitions [{
      sphereCollisionRadius=0
      boundBone {boneName=dyn_armVialE_Liquid}
      boxExtents=<10 10 2>
      constraintSetup {
       angularTargetAxis=AxisZ
       twistAxis=AxisZ
       angularLimitsMax=<28 28 0>
       angularLimitsMin=<-28 -28 0>
       angularTarget=<0 0 0>
       linearAxesMax=<0 0 1>
      }
      localJointOffset=<0 0 -2>
     }]
   }
   n38 = Anim Dynamics - Bone: dyn_legVialA(ComponentPose=@n37, Alpha=1, ExternalForce="0,0,0") {
    angularDampingOverride=0.999
    angularSpringConstant=1000
    bAngularSpring=true
    bGravityOverrideInSimSpace=true
    bOverrideAngularDamping=true
    bUseGravityOverride=true
    linearSpringConstant=1000
    simulationSpace=BoneRelative
    alphaBoolBlend {blendOption=Linear}
    boundBone {boneName=dyn_legVialA}
    componentAppliedLinearAccClamp=<800 800 800>
    componentLinearAccScale=<1 1 1>
    componentLinearVelScale=<1 1 1>
    gravityOverride {z=-980}
    physicsBodyDefinitions [{
      sphereCollisionRadius=0
      boundBone {boneName=dyn_legVialA}
      boxExtents=<14 16 14>
      constraintSetup {angularLimitsMax=<12 8 0>, angularLimitsMin=<-12 -8 0>, angularTarget=<0 0 0>}
      localJointOffset=<-4 0 0>
     }]
    relativeSpaceBone {boneName=pelvis}
   }
   n39 = Anim Dynamics - Bone: dyn_legVialA_Liquid(ComponentPose=@n38, Alpha=1, ExternalForce="0,0,0") {
    angularDampingOverride=0.95
    angularSpringConstant=1000
    bAngularSpring=true
    bGravityOverrideInSimSpace=true
    bOverrideAngularDamping=true
    bUseGravityOverride=true
    simulationSpace=BoneRelative
    alphaBoolBlend {blendOption=Linear}
    boundBone {boneName=dyn_legVialA_Liquid}
    componentAppliedLinearAccClamp=<3000 3000 3000>
    componentLinearAccScale=<0.5 0.5 0.5>
    componentLinearVelScale=<0.5 0.5 0.5>
    gravityOverride {z=-980}
    physicsBodyDefinitions [{
      sphereCollisionRadius=0
      boundBone {boneName=dyn_legVialA_Liquid}
      boxExtents=<10 10 2>
      constraintSetup {
       angularTargetAxis=AxisZ
       twistAxis=AxisZ
       angularLimitsMax=<28 28 0>
       angularLimitsMin=<-28 -28 0>
       angularTarget=<0 0 0>
       linearAxesMax=<0 0 0.8>
      }
      localJointOffset=<0 0 -2>
     }]
   }
   n40 = Anim Dynamics - Bone: dyn_legVialB(ComponentPose=@n39, Alpha=1, ExternalForce="0,0,0") {
    angularDampingOverride=0.999
    angularSpringConstant=1000
    bAngularSpring=true
    bGravityOverrideInSimSpace=true
    bOverrideAngularDamping=true
    bUseGravityOverride=true
    simulationSpace=BoneRelative
    alphaBoolBlend {blendOption=Linear}
    boundBone {boneName=dyn_legVialB}
    componentAppliedLinearAccClamp=<800 800 800>
    componentLinearAccScale=<1 1 1>
    componentLinearVelScale=<1 1 1>
    gravityOverride {z=-980}
    physicsBodyDefinitions [{
      sphereCollisionRadius=0
      boundBone {boneName=dyn_legVialB}
      boxExtents=<14 21 14>
      constraintSetup {angularLimitsMax=<12 12 0>, angularLimitsMin=<-12 -12 0>, angularTarget=<0 0 0>}
      localJointOffset=<-4.5 0 0>
     }]
    relativeSpaceBone {boneName=pelvis}
   }
   n41 = Anim Dynamics - Bone: dyn_legVialB_Liquid(ComponentPose=@n40, Alpha=1, ExternalForce="0,0,0") {
    angularDampingOverride=0.95
    angularSpringConstant=1000
    bAngularSpring=true
    bGravityOverrideInSimSpace=true
    bOverrideAngularDamping=true
    bUseGravityOverride=true
    simulationSpace=BoneRelative
    alphaBoolBlend {blendOption=Linear}
    boundBone {boneName=dyn_legVialB_Liquid}
    componentAppliedLinearAccClamp=<3000 3000 3000>
    componentLinearAccScale=<0.5 0.5 0.5>
    componentLinearVelScale=<0.5 0.5 0.5>
    gravityOverride {z=-980}
    physicsBodyDefinitions [{
      sphereCollisionRadius=0
      boundBone {boneName=dyn_legVialB_Liquid}
      boxExtents=<20 20 2>
      constraintSetup {
       angularTargetAxis=AxisZ
       twistAxis=AxisZ
       angularLimitsMax=<15 10 0>
       angularLimitsMin=<-15 -10 0>
       angularTarget=<0 0 0>
       linearAxesMax=<0 0 0.5>
      }
      localJointOffset=<0 0 -1>
     }]
   }
   n42 = Anim Dynamics - Bone: dyn_legVialC(ComponentPose=@n41, Alpha=1, ExternalForce="0,0,0") {
    angularDampingOverride=0.999
    angularSpringConstant=1000
    bAngularSpring=true
    bGravityOverrideInSimSpace=true
    bOverrideAngularDamping=true
    bUseGravityOverride=true
    simulationSpace=BoneRelative
    alphaBoolBlend {blendOption=Linear}
    boundBone {boneName=dyn_legVialC}
    componentAppliedLinearAccClamp=<800 800 800>
    componentLinearAccScale=<1 1 1>
    componentLinearVelScale=<1 1 1>
    gravityOverride {z=-980}
    physicsBodyDefinitions [{
      sphereCollisionRadius=0
      boundBone {boneName=dyn_legVialC}
      boxExtents=<18 22 18>
      constraintSetup {angularLimitsMax=<12 10 0>, angularLimitsMin=<-12 -10 0>, angularTarget=<0 0 0>}
      localJointOffset=<-5 0 0>
     }]
    relativeSpaceBone {boneName=pelvis}
   }
   n43 = Anim Dynamics - Bone: dyn_legVialC_Liquid(ComponentPose=@n42, Alpha=1, ExternalForce="0,0,0") {
    angularDampingOverride=0.95
    angularSpringConstant=1000
    bAngularSpring=true
    bGravityOverrideInSimSpace=true
    bOverrideAngularDamping=true
    bUseGravityOverride=true
    simulationSpace=BoneRelative
    alphaBoolBlend {blendOption=Linear}
    boundBone {boneName=dyn_legVialC_Liquid}
    componentAppliedLinearAccClamp=<3000 3000 3000>
    componentLinearAccScale=<0.5 0.5 0.5>
    componentLinearVelScale=<0.5 0.5 0.5>
    gravityOverride {z=-980}
    physicsBodyDefinitions [{
      sphereCollisionRadius=0
      boundBone {boneName=dyn_legVialC_Liquid}
      boxExtents=<10 10 2>
      constraintSetup {
       angularTargetAxis=AxisZ
       twistAxis=AxisZ
       angularLimitsMax=<33 33 0>
       angularLimitsMin=<-33 -33 0>
       angularTarget=<0 0 0>
       linearAxesMax=<0 0 1>
      }
      localJointOffset=<0 0 -2>
     }]
   }
   n44 = Anim Dynamics - Bone: dyn_chestVial(ComponentPose=@n43, Alpha=1, ExternalForce="0,0,0") {
    angularDampingOverride=0.999
    angularSpringConstant=500
    bAngularSpring=true
    bGravityOverrideInSimSpace=true
    bOverrideAngularDamping=true
    bOverrideLinearDamping=true
    bUseGravityOverride=true
    gravityScale=0.6
    linearDampingOverride=0.999
    linearSpringConstant=50
    alphaBoolBlend {blendOption=Linear}
    boundBone {boneName=dyn_chestVial}
    componentAppliedLinearAccClamp=<1000 1000 1000>
    componentLinearAccScale=<1 1 1>
    componentLinearVelScale=<1 1 1>
    gravityOverride {z=980}
    physicsBodyDefinitions [{
      sphereCollisionRadius=0
      boundBone {boneName=dyn_chestVial}
      boxExtents=<45 45 11>
      constraintSetup {twistAxis=AxisZ, angularLimitsMax=<15 2 0>, angularLimitsMin=<-15 -2 -44>, angularTarget=<0 0 0>}
      localJointOffset=<-3 0 -0.5>
     }]
   }
   n45 = Anim Dynamics - Bone: dyn_chestVial_Liquid(ComponentPose=@n44, Alpha=1, ExternalForce="0,0,0") {
    angularDampingOverride=0.999
    angularSpringConstant=1000
    bAngularSpring=true
    bGravityOverrideInSimSpace=true
    bOverrideAngularDamping=true
    bUseGravityOverride=true
    linearSpringConstant=100
    simulationSpace=BoneRelative
    alphaBoolBlend {blendOption=Linear}
    boundBone {boneName=dyn_chestVial_Liquid}
    componentAppliedLinearAccClamp=<3000 3000 3000>
    componentLinearAccScale=<0.8 0.8 0.8>
    componentLinearVelScale=<0.8 0.8 0.8>
    gravityOverride {z=-980}
    physicsBodyDefinitions [{
      sphereCollisionRadius=0
      boundBone {boneName=dyn_chestVial_Liquid}
      boxExtents=<15 15 2>
      constraintSetup {
       angularTargetAxis=AxisZ
       twistAxis=AxisZ
       angularLimitsMax=<10 10 0>
       angularLimitsMin=<-10 -10 0>
       angularTarget=<0 0 0>
       linearAxesMax=<0.01 0.01 1>
      }
      localJointOffset=<0 0 -2>
     }]
   }
   n46 = RigidBody(ComponentPose=@n45) {
    bClampLinearTranslationLimitToRefPose=true
    bOverrideWorldGravity=true
    overridePhysicsAsset=&/M_MED_Freak_CoatMass_Physics
    simulationSpace=BaseBoneSpace
    alphaBoolBlend {blendOption=Linear}
    baseBoneRef {boneName=spine_03}
    componentAppliedLinearAccClamp=<4000 4000 4000>
    componentLinearAccScale=<0.4 0.4 0.5>
    componentLinearVelScale=<0.2 0.2 0.3>
    overrideWorldGravity {z=-980}
   }
   Output Pose(Result=Component To Local(ComponentPose=@n46)) {layerGroup=None}
  ]}
 EventGraph {nodes [
   n1 = Event BlueprintUpdateAnimation()
   > n2 = Sequence()
   n3 = "To Float (Boolean)"(InBool=Get bIsCrouching())
   > Set crouchAlpha(crouchAlpha:double=FInterpTo(Current=@n3, Target=@n3, DeltaTime=@n1.DeltaTimeX, InterpSpeed=1)) -> (Output_Get:double)
   n4 = Get bIsJumping()
   n5 = Get bIsFalling()
   > Set jumpAlpha(jumpAlpha:double="To Float (Boolean)"(InBool=OR Boolean(A=@n5, B=@n4))) -> (Output_Get:double)
   n6 = Branch(execute=@n2.then_1, Condition=Get bIsTargeting())
   > Set SwingDelta(SwingDelta:double=FInterpTo(Current=Get SwingDelta(), Target=1, DeltaTime=@n1.DeltaTimeX, InterpSpeed=5)) -> (Output_Get:double)
   Set SwingDelta(execute=@n6.else, SwingDelta=FInterpTo(Current=Get SwingDelta(), Target=0, DeltaTime=@n1.DeltaTimeX, InterpSpeed=2))
   TryGetPawnOwner()
  ]}
]
variables [
 crouchAlpha {type=number}
 jumpAlpha {type=number}
 SwingDelta {type=number}
]
