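# Control Rig for the Freak character's face accessories, animating the long tongue and hood bones.
# It offsets tongue bones, then adds procedural wave motion with ChainHarmonics from dyn_tongue_01.
# The `targeting` variable weights added rotations on neck, head, tongue base and left hood base.
# Two parent constraints both drive dyn_hood_Base_l, blending head with left, then right clavicle; dyn_hood_Base_r is untouched.
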ControlRigBlueprint /BRCosmetics/Characters/Player/Male/Medium/Bodies/M_MED_Freak/Meshes/Parts/M_MED_Freak_FaceAcc_CtrlRig
graphs [
 RigVMFunctionLibrary {links [], nodes []}
 RigVMModel {nodes [
   n1 = &RigUnit_BeginExecution()
   &RigUnit_ModifyTransforms(
    ItemToModify="((Item=(Type=Bone,Name=\"dyn_tongue_01_Base\"),Transform=(Rotation=(X=0,Y=0,Z=0,W=1),Translation=(X=0,Y=0,Z=0),Scale3D=(X=1,Y=1,Z=1))),(Item=(Type=Bone,Name=\"dyn_tongue_02\"),Transform=(Rotation=(X=0,Y=0,Z=0,W=1),Translation=(X=2,Y=0,Z=0),Scale3D=(X=1,Y=1,Z=1 ...(truncated)"
    Weight=1
    WeightMinimum=0
    ExecutePin=@n1.ExecutePin
    ItemToModify.0.Transform.Rotation=&RigVMFunction_MathQuaternionFromEuler(Euler=(X=0,Y=0,Z=-10)).Result
   )
   > /Script/ControlRig.RigUnit_ChainHarmonicsPerItem(
    ChainRoot=(Type=Bone,Name="dyn_tongue_01")
    DrawWorldOffset=(Rotation=(X=0,Y=0,Z=0,W=1),Translation=(X=0,Y=0,Z=0),Scale3D=(X=1,Y=1,Z=1))
    Pendulum=(bEnabled=False,PendulumStiffness=2,PendulumGravity=(X=0,Y=0,Z=0),PendulumBlend=0.75,PendulumDrag=0.98,PendulumMinimum=0,PendulumMaximum=0.1,PendulumEase=Linear,UnwindAxis=(X=0,Y=1,Z=0),UnwindMinimum=0.2,UnwindMaximum=0.05)
    Reach=(bEnabled=False,ReachTarget=(X=0,Y=0,Z=0),ReachAxis=(X=1,Y=0,Z=0),ReachMinimum=0,ReachMaximum=0,ReachEase=SineEaseInOut)
    Speed=(X=0.25,Y=0.25,Z=0.1)
    Wave=(bEnabled=True,WaveFrequency=(X=0.2,Y=0.2,Z=0.2),WaveAmplitude=(X=0.25,Y=2.5,Z=1.5),WaveOffset=(X=1,Y=0,Z=2),WaveNoise=(X=8,Y=8,Z=30),WaveMinimum=0,WaveMaximum=0.75,WaveEase=SineEaseIn)
    WaveCurve=(EditorCurveData=(Keys=((),(Time=1,Value=1)),DefaultValue=340282346638528859811704183484516925440,PreInfinityExtrap=RCCE_Constant,PostInfinityExtrap=RCCE_Constant),ExternalCurve=None)
    bDrawDebug=True
   ) {is=struct RigVMUnitNode}
   n2 = &RigVMFunction_MathQuaternionFromEuler(Euler=(X=-5,Y=0,Z=0))
   n3 = RigVMVariableNode() {variable=targeting}
   > &RigUnit_ModifyTransforms(
    ItemToModify=((Item=(Type=Bone,Name="neck_01"),Transform=(Rotation=(X=0,Y=0,Z=0,W=1),Translation=(X=0,Y=0,Z=0),Scale3D=(X=1,Y=1,Z=1))))
    WeightMinimum=0
    ItemToModify.0.Transform.Rotation=@n2.Result
    Weight=@n3.Value
   )
   > &RigUnit_ModifyTransforms(
    ItemToModify=((Item=(Type=Bone,Name="neck_02"),Transform=(Rotation=(X=0,Y=0,Z=0,W=1),Translation=(X=0,Y=0,Z=0),Scale3D=(X=1,Y=1,Z=1))))
    WeightMinimum=0
    ItemToModify.0.Transform.Rotation=@n2.Result
    Weight=@n3.Value
   )
   > &RigUnit_ModifyTransforms(
    ItemToModify=((Item=(Type=Bone,Name="head"),Transform=(Rotation=(X=0,Y=0,Z=0,W=1),Translation=(X=0,Y=0,Z=0),Scale3D=(X=1,Y=1,Z=1))))
    WeightMinimum=0
    ItemToModify.0.Transform.Rotation=@n2.Result
    Weight=@n3.Value
   )
   > &RigUnit_ModifyTransforms(
    ItemToModify=((Item=(Type=Bone,Name="dyn_tongue_01_Base"),Transform=(Rotation=(X=0,Y=0,Z=0,W=1),Translation=(X=0,Y=0,Z=0),Scale3D=(X=1,Y=1,Z=1))))
    WeightMinimum=0
    ItemToModify.0.Transform.Rotation=&RigVMFunction_MathQuaternionFromEuler(Euler=(X=-10,Y=30,Z=0)).Result
    Weight=@n3.Value
   )
   > &RigUnit_ParentConstraint(
    AdvancedSettings=(InterpolationType=Average,RotationOrderForFilter=XZY)
    Child=(Type=Bone,Name="dyn_hood_Base_l")
    Parents=((Item=(Type=Bone,Name="head"),Weight=0.8),(Item=(Type=Bone,Name="clavicle_l"),Weight=0.2))
    Weight=1
   )
   > &RigUnit_ParentConstraint(
    AdvancedSettings=(InterpolationType=Average,RotationOrderForFilter=XZY)
    Child=(Type=Bone,Name="dyn_hood_Base_l")
    Parents=((Item=(Type=Bone,Name="head"),Weight=0.7),(Item=(Type=Bone,Name="clavicle_r"),Weight=0.3))
    Weight=1
   )
   > &RigUnit_ModifyTransforms(
    ItemToModify=((Item=(Type=Bone,Name="dyn_hood_Base_l"),Transform=(Rotation=(X=0,Y=0,Z=0,W=1),Translation=(X=0,Y=0,Z=0),Scale3D=(X=1,Y=1,Z=1))))
    WeightMinimum=0
    ItemToModify.0.Transform.Rotation=&RigVMFunction_MathQuaternionFromEuler(Euler=(X=50,Y=0,Z=0)).Result
    Weight=@n3.Value
   )
  ]}
]
hierarchy [
 root {loc=<0 0 0>, rot=<0 0 0 1>}
 attach {parent=root, loc=<0 0 0>, rot=<0 0 0 1>}
 pelvis {parent=root, loc=<0 -0.7071 93.1191>, rot=<0.0127 0.707 -0.0127 -0.707>}
 spine_01 {parent=pelvis, loc=<6.7494 -0.6388 0>, rot=<0 0 0.034 -0.9994>}
 spine_02 {parent=spine_01, loc=<7.2518 -0.035 0>, rot=<0 0 0.0238 -0.9997>}
 spine_03 {parent=spine_02, loc=<6.0104 0 0>, rot=<0 0 0.0149 -0.9999>}
 spine_04 {parent=spine_03, loc=<8.1024 0 0>, rot=<0 0 -0.0132 -0.9999>}
 spine_05 {parent=spine_04, loc=<10.0146 0 0>, rot=<0 0 -0.0015 -1>}
 clavicle_l {parent=spine_05, loc=<4.3575 -0.1466 -2.3223>, rot=<0.0465 -0.7289 -0.009 -0.6829>}
 upperarm_l {parent=clavicle_l, loc=<15.5111 0 0>, rot=<0.0226 -0.3234 -0.0088 -0.9459>}
 upperarm_twist_01_l {parent=upperarm_l, loc=<0.0041 0 0>, rot=<0 0 0 -1>}
 upperarm_twist_02_l {parent=upperarm_l, loc=<13.7941 0 0>, rot=<0 0 0 -1>}
 dyn_armVialA_Base {parent=upperarm_twist_02_l, loc=<1.1839 4.8099 4.2179>, rot=<0.8417 0.0282 -0.0183 -0.5388>}
 dyn_armVialA {parent=dyn_armVialA_Base, loc=<0 0 0>, rot=<0 0 0 -1>}
 dyn_armVialA_Liquid {parent=dyn_armVialA, loc=<0.7605 -0.0092 -0.0539>, rot=<-0.7877 0.1772 0.3091 -0.5026>}
 dyn_armVialB_Base {parent=upperarm_twist_02_l, loc=<1.2033 1.3192 5.3479>, rot=<0.7034 0.0174 -0.0118 -0.7105>}
 dyn_armVialB {parent=dyn_armVialB_Base, loc=<0 0 0>, rot=<0 0 0 -1>}
 dyn_armVialB_Liquid {parent=dyn_armVialB, loc=<0.7067 -0.0003 0.0018>, rot=<-0.6628 0.2463 0.2514 -0.6609>}
 dyn_armVialC_Base {parent=upperarm_twist_02_l, loc=<1.6471 -2.81 4.6642>, rot=<0.5648 -0.0063 -0.0219 -0.8249>}
 dyn_armVialC {parent=dyn_armVialC_Base, loc=<0 0 0>, rot=<0 0 0 1>}
 dyn_armVialC_Liquid {parent=dyn_armVialC, loc=<0.8722 -0.0078 0.0593>, rot=<-0.5309 0.3073 0.2096 -0.7614>}
 dyn_armVialD_Base {parent=upperarm_twist_02_l, loc=<1.9969 -6.6955 2.1794>, rot=<0.3809 -0.0107 -0.0201 -0.9243>}
 dyn_armVialD {parent=dyn_armVialD_Base, loc=<0 0 0>, rot=<0 0 0 1>}
 dyn_armVialD_Liquid {parent=dyn_armVialD, loc=<1.15 -0.0159 0.0187>, rot=<-0.361 0.3441 0.1411 -0.8552>}
 dyn_armVialE_Base {parent=upperarm_twist_02_l, loc=<2.1072 -8.1169 -2.2451>, rot=<0.0642 -0.0057 -0.0396 -0.9971>}
 dyn_armVialE {parent=dyn_armVialE_Base, loc=<0 0 0>, rot=<0 0 0 1>}
 dyn_armVialE_Liquid {parent=dyn_armVialE, loc=<1.6047 -0.0206 0.027>, rot=<-0.0592 0.3612 0.0458 -0.9295>}
 clavicle_r {parent=spine_05, loc=<4.3575 -0.1466 2.3223>, rot=<0.7289 0.0465 -0.6829 0.009>}
 upperarm_r {parent=clavicle_r, loc=<-15.5111 0 0>, rot=<0.0226 -0.3234 -0.0088 -0.9459>}
 upperarm_twist_01_r {parent=upperarm_r, loc=<-0.0041 0 0>, rot=<0 0 0 1>}
 upperarm_twist_02_r {parent=upperarm_r, loc=<-13.7942 0 0>, rot=<0 0 0 1>}
 neck_01 {parent=spine_05, loc=<11.3138 0.1529 0>, rot=<0 0 -0.2003 -0.9797>}
 neck_02 {parent=neck_01, loc=<4.5627 0 0>, rot=<0 0 0.0504 -0.9987>}
 head {parent=neck_02, loc=<5.3914 0 0>, rot=<0 0 0.115 -0.9934>}
 faceAttach {parent=head, loc=<-1.0179 -0.1338 -0.016>, rot=<0.0082 -0.7071 -0.0082 -0.7071>, scale=<0.89 0.89 0.89>}
 C_brow_mid {parent=faceAttach, loc=<-0.0167 10.9624 6.5322>, rot=<0 0.0201 -0.004 -0.9998>}
 L_brow_mid {parent=faceAttach, loc=<3.6016 10.4388 8.5337>, rot=<-0.0002 -0.0046 0.2553 -0.9669>}
 L_brow_outer {parent=faceAttach, loc=<5.9089 7.5398 8.7873>, rot=<0.0651 0.0781 0.4251 -0.8994>}
 L_cheek_inner {parent=faceAttach, loc=<4.0251 8.9381 2.1034>, rot=<-0.0513 -0.2208 0.3248 -0.9182>}
 C_lip_upper_mid {parent=faceAttach, loc=<-0.0095 11.433 0.2005>, rot=<0.2225 -0.0001 0.0001 -0.9749>}
 L_lip_upper_outer {parent=faceAttach, loc=<1.6109 10.6339 0.0984>, rot=<0.0221 -0.1115 0.2771 -0.9541>}
 L_lip_corner {parent=faceAttach, loc=<2.6536 9.9366 -0.1543>, rot=<0.003 0.008 -0.0146 -0.9999>}
 C_nose_bridge {parent=faceAttach, loc=<-0.0172 12.5105 3.7835>, rot=<-0.2218 -0.0048 -0.0014 -0.9751>}
 R_brow_mid {parent=faceAttach, loc=<-3.6353 10.439 8.5383>, rot=<-0.0001 0.0043 -0.2553 -0.9669>}
 R_brow_outer {parent=faceAttach, loc=<-5.9654 7.5674 8.7426>, rot=<0.0463 -0.0543 -0.4307 -0.8996>}
 R_cheek_inner {parent=faceAttach, loc=<-4.0476 8.9458 2.0974>, rot=<-0.0523 0.2233 -0.3245 -0.9177>}
 R_lip_corner {parent=faceAttach, loc=<-2.6766 9.9699 -0.125>, rot=<0.003 -0.008 0.0146 -0.9999>}
 C_jaw {parent=faceAttach, loc=<-0.051 3.2543 0.4291>, rot=<0 0 0 1>}
 L_lip_lower_outer {parent=C_jaw, loc=<1.7244 7.3148 -0.3582>, rot=<0.024 0.1045 0.3502 -0.9305>}
 C_lip_lower_mid {parent=C_jaw, loc=<0.0403 8.0838 -0.4699>, rot=<0 -0.022 -0.0047 -0.9997>}
 R_lip_lower_outer {parent=C_jaw, loc=<-1.7934 7.4704 -0.4893>, rot=<0.0143 -0.0515 -0.3109 -0.9489>}
 teeth_lower {parent=C_jaw, loc=<0.0733 5.168 -2.0669>, rot=<0 0 0 1>}
 tongue {parent=C_jaw, loc=<0.0359 1.7722 -0.7013>, rot=<-0.1568 0 0 -0.9876>}
 R_eye {parent=faceAttach, loc=<-3.1018 8.243 6.4184>, rot=<0 0 0.9998 -0.0187>}
 R_eye_lid_upper_mid {parent=R_eye, loc=<-0.0383 -1.2865 0.1479>, rot=<0.0523 -0.0009 0.0167 -0.9985>}
 R_eye_lid_lower_mid {parent=R_eye, loc=<0.0536 -1.0737 -0.8142>, rot=<0.0329 -0.0005 0.0167 -0.9993>}
 L_eye {parent=faceAttach, loc=<3.0658 8.243 6.4184>, rot=<0 0 0.9998 0.0187>}
 L_eye_lid_upper_mid {parent=L_eye, loc=<0.0202 -1.2607 0.0938>, rot=<0.0523 0 0 -0.9986>}
 L_eye_lid_lower_mid {parent=L_eye, loc=<-0.0383 -1.0554 -0.8584>, rot=<0.0329 0 0 -0.9995>}
 teeth_upper {parent=faceAttach, loc=<-0.0174 9.5667 0.8192>, rot=<0 0 0 1>}
 R_lip_upper_outer {parent=faceAttach, loc=<-1.6682 10.702 0.072>, rot=<0.0213 0.1138 -0.2779 -0.9536>}
 dyn_hood_01_Base {parent=head, loc=<9.8083 1.0572 -0.3391>, rot=<0.7894 -0.6103 0.0407 0.0517>}
 dyn_hood_01 {parent=dyn_hood_01_Base, loc=<0 0 0>, rot=<0 0 0 1>}
 dyn_hood_02 {parent=dyn_hood_01, loc=<6.5966 0 0>, rot=<-0.0165 0.1015 -0.085 -0.9911>}
 dyn_hood_03 {parent=dyn_hood_02, loc=<7.554 0 0>, rot=<-0.0044 -0.0618 -0.1455 -0.9874>}
 dyn_hood_Base_r {parent=head, loc=<0.1708 6.6565 12.9545>, rot=<-0.0132 0.0541 0.972 0.2282>}
 dyn_hood_r {parent=dyn_hood_Base_r, loc=<0 0 0>, rot=<0 0 0 1>}
 dyn_hood_Base_l {parent=head, loc=<0.1251 7.765 -12.2723>, rot=<0.193 0.5678 0.7883 0.1374>}
 dyn_hood_l {parent=dyn_hood_Base_l, loc=<0 0 0>, rot=<0 0 0 1>}
 dyn_hoodTeeth_01_Base {parent=head, loc=<8.6584 9.6741 -0.5378>, rot=<0.0705 -0.0377 0.8798 0.4686>}
 dyn_hoodTeeth_01 {parent=dyn_hoodTeeth_01_Base, loc=<0 0 0>, rot=<0 0 0 1>}
 dyn_hoodTeeth_02 {parent=dyn_hoodTeeth_01, loc=<8.0529 0 0>, rot=<-0.0419 -0.1013 -0.2163 -0.9702>}
 dyn_tongue_01_Base {parent=spine_05, loc=<10.3139 6.9125 0>, rot=<0 0 -0.8288 -0.5596>}
 dyn_tongue_01 {parent=dyn_tongue_01_Base, loc=<0 0 0>, rot=<0 0 0 -1>}
 dyn_tongue_02 {parent=dyn_tongue_01, loc=<5.9402 0 0>, rot=<0 0 -0.2765 -0.961>}
 dyn_tongue_03 {parent=dyn_tongue_02, loc=<8.0944 0 0>, rot=<0 0 -0.1669 -0.986>}
 dyn_tongue_04 {parent=dyn_tongue_03, loc=<7.5277 0 0>, rot=<0 0 -0.0768 -0.997>}
 dyn_tongue_05 {parent=dyn_tongue_04, loc=<4.3252 0 0>, rot=<0 0 -0.0072 -1>}
 dyn_tongue_06 {parent=dyn_tongue_05, loc=<4.3508 0 0>, rot=<0 0 -0.0314 -0.9995>}
 dyn_tongue_07 {parent=dyn_tongue_06, loc=<4.3883 0 0>, rot=<0 0 -0.0033 -1>}
 dyn_tongue_08 {parent=dyn_tongue_07, loc=<4.1829 0 0>, rot=<0 0 0 1>}
 dyn_chestBoneB_Base {parent=spine_02, loc=<3.0245 16.015 0.2524>, rot=<-0.0252 0.0359 0.9826 0.1804>}
 dyn_chestBoneB {parent=dyn_chestBoneB_Base, loc=<0 0 0>, rot=<0 0 0 1>}
 dyn_chestBoneA_Base {parent=spine_02, loc=<5.0655 15.7656 3.9879>, rot=<0.0722 -0.0148 0.9908 0.1138>}
 dyn_chestBoneA {parent=dyn_chestBoneA_Base, loc=<0 0 0>, rot=<0 0 0 1>}
 dyn_chestVial_Base {parent=spine_01, loc=<2.7547 14.4996 -10.6117>, rot=<0.06 0.2082 0.9534 0.2101>}
 dyn_chestVial {parent=dyn_chestVial_Base, loc=<0 0 0>, rot=<0 0 0 1>}
 dyn_chestVial_Liquid {parent=dyn_chestVial, loc=<6.0027 -0.7602 0.273>, rot=<-0.6269 -0.0129 0.7165 -0.3059>}
 dyn_skirtFront_01_Base_r {parent=pelvis, loc=<-2.3846 18.7438 5.0223>, rot=<0.069 -0.994 -0.0089 0.0846>}
 dyn_skirtFront_01_r {parent=dyn_skirtFront_01_Base_r, loc=<0 0 0>, rot=<0 0 0 1>}
 dyn_skirtFront_02_r {parent=dyn_skirtFront_01_r, loc=<8.1313 0 0>, rot=<-0.0122 0.0587 0.0308 -0.9977>}
 dyn_skirtFront_03_r {parent=dyn_skirtFront_02_r, loc=<12.3887 0 0>, rot=<-0.0854 0.7196 -0.0812 -0.6843>}
 dyn_skirtFrontSide_01_Base_r {parent=pelvis, loc=<-2.1914 12.9709 14.5202>, rot=<-0.0452 -0.997 0.0006 0.0626>}
 dyn_skirtFrontSide_01_r {parent=dyn_skirtFrontSide_01_Base_r, loc=<0 0 0>, rot=<0 0 0 -1>}
 dyn_skirtFrontSide_02_r {parent=dyn_skirtFrontSide_01_r, loc=<6.9305 0 0>, rot=<-0.006 0.1023 0.1129 -0.9883>}
 dyn_skirtFrontSide_03_r {parent=dyn_skirtFrontSide_02_r, loc=<13.067 0 0>, rot=<-0.0584 0.6757 -0.0633 -0.7321>}
 dyn_skirtSide_01_Base_r {parent=pelvis, loc=<-2.0353 5.1823 19.4785>, rot=<-0.0168 -0.9969 -0.0015 0.077>}
 dyn_skirtSide_01_r {parent=dyn_skirtSide_01_Base_r, loc=<0 0 0>, rot=<0 0 0 -1>}
 dyn_skirtSide_02_r {parent=dyn_skirtSide_01_r, loc=<7.2939 0 0>, rot=<-0.004 0.1743 0.0202 -0.9845>}
 dyn_skirtSide_03_r {parent=dyn_skirtSide_02_r, loc=<12.8759 0 0>, rot=<-0.0137 0.6342 -0.0167 -0.7728>}
 dyn_skirtFront_01_Base_l {parent=pelvis, loc=<-1.7708 16.2018 -6.9891>, rot=<-0.0182 -0.9995 0.0005 -0.027>}
 dyn_skirtFront_01_l {parent=dyn_skirtFront_01_Base_l, loc=<0 0 0>, rot=<0 0 0 -1>}
 dyn_skirtFront_02_l {parent=dyn_skirtFront_01_l, loc=<7.7057 0 0>, rot=<0.0021 -0.0192 0.1095 -0.9938>}
 dyn_skirtFront_03_l {parent=dyn_skirtFront_02_l, loc=<12.9124 0 0>, rot=<-0.0766 0.6974 -0.0778 -0.7083>}
 dyn_skirtFrontSide_01_Base_l {parent=pelvis, loc=<-1.9386 11.1843 -14.3485>, rot=<-0.034 -0.9989 0.0001 -0.0334>}
 dyn_skirtFrontSide_01_l {parent=dyn_skirtFrontSide_01_Base_l, loc=<0 0 0>, rot=<0 0 0 -1>}
 dyn_skirtFrontSide_02_l {parent=dyn_skirtFrontSide_01_l, loc=<7.8329 0 0>, rot=<0.0039 -0.0734 0.085 -0.9937>}
 dyn_skirtFrontSide_03_l {parent=dyn_skirtFrontSide_02_l, loc=<11.812 0 0>, rot=<-0.0508 0.7332 -0.0469 -0.6765>}
 dyn_skirtSide_01_Base_l {parent=pelvis, loc=<-2.0754 4.8295 -18.88>, rot=<-0.0388 -0.9987 -0.0001 -0.0335>}
 dyn_skirtSide_01_l {parent=dyn_skirtSide_01_Base_l, loc=<0 0 0>, rot=<0 0 0 -1>}
 dyn_skirtSide_02_l {parent=dyn_skirtSide_01_l, loc=<7.8229 0 0>, rot=<0 -0.1101 0.0416 -0.9931>}
 dyn_skirtSide_03_l {parent=dyn_skirtSide_02_l, loc=<11.7634 0 0>, rot=<-0.0159 0.7591 -0.0137 -0.6506>}
 dyn_chestTieTop_Base {parent=pelvis, loc=<-0.6047 17.9748 1.4413>, rot=<-0.0127 -0.707 0.0127 -0.707>}
 dyn_chestTieTop {parent=dyn_chestTieTop_Base, loc=<0 0 0>, rot=<0 0 0 1>}
 dyn_chestTieBottom_Base {parent=pelvis, loc=<-7.2534 17.1565 -0.8926>, rot=<-0.0127 -0.707 0.0127 -0.707>}
 dyn_chestTieBottom {parent=dyn_chestTieBottom_Base, loc=<0 0 0>, rot=<0 0 0 1>}
 dyn_frogMain_Base {parent=pelvis, loc=<5.7618 10.9043 -18.1343>, rot=<0.1772 -0.3372 -0.6261 -0.6804>}
 dyn_frogMain {parent=dyn_frogMain_Base, loc=<0 0 0>, rot=<0 0 0 -1>}
 dyn_frogHead {parent=dyn_frogMain, loc=<1.0031 -4.8709 -0.0751>, rot=<-0.0179 0.0084 -0.202 -0.9792>}
 dyn_frogTongue_01 {parent=dyn_frogHead, loc=<3.5212 -3.115 3.6533>, rot=<0.3401 0.1988 -0.3144 -0.8637>}
 dyn_frogTongue_02 {parent=dyn_frogTongue_01, loc=<1.6397 0 0>, rot=<0.01 -0.0704 -0.2103 -0.975>}
 dyn_frogArm_01_r {parent=dyn_frogMain, loc=<2.0443 -2.2586 4.8662>, rot=<0.1173 0.4137 -0.2353 -0.8716>}
 dyn_frogArm_02_r {parent=dyn_frogArm_01_r, loc=<2.7753 0 0>, rot=<0.7046 0.4678 -0.2175 0.4872>}
 dyn_frogArm_03_r {parent=dyn_frogArm_02_r, loc=<4.8543 0 0>, rot=<0.7431 0 0 -0.6691>}
 dyn_frogArm_01_l {parent=dyn_frogMain, loc=<2.0327 -2.1504 -5.1483>, rot=<-0.0505 -0.3737 -0.2924 -0.8788>}
 dyn_frogArm_02_l {parent=dyn_frogArm_01_l, loc=<2.4422 0 0>, rot=<0.7069 0.3928 0.2927 -0.5103>}
 dyn_frogArm_03_l {parent=dyn_frogArm_02_l, loc=<4.8258 0 0>, rot=<-0.788 0 0 -0.6157>}
 dyn_frogBelly {parent=dyn_frogMain, loc=<1.296 2.1005 0>, rot=<0 0 -0.6293 -0.7771>}
 dyn_frogLeg_01_l {parent=dyn_frogBelly, loc=<4.3431 0.4916 -3.4506>, rot=<-0.1776 -0.3115 0.1507 -0.9212>}
 dyn_frogLeg_02_l {parent=dyn_frogLeg_01_l, loc=<3.2675 0 0>, rot=<0.0038 0.0593 -0.6287 -0.7754>}
 dyn_frogLeg_03_l {parent=dyn_frogLeg_02_l, loc=<3.8381 0 0>, rot=<0.1491 0.1944 0.6756 -0.6954>}
 dyn_frogLeg_04_l {parent=dyn_frogLeg_03_l, loc=<4.9618 0 0>, rot=<-0.0306 -0.0332 -0.2519 -0.9667>}
 dyn_frogLeg_05_l {parent=dyn_frogLeg_04_l, loc=<2.7439 0 0>, rot=<-0.5537 0.2872 0.6189 -0.4774>}
 dyn_frogLeg_01_r {parent=dyn_frogBelly, loc=<4.4002 0.3581 3.3911>, rot=<0.2182 0.3195 0.1814 -0.9041>}
 dyn_frogLeg_02_r {parent=dyn_frogLeg_01_r, loc=<3.2446 0 0>, rot=<-0.0618 -0.0505 -0.6398 -0.7644>}
 dyn_frogLeg_03_r {parent=dyn_frogLeg_02_r, loc=<3.8697 0 0>, rot=<-0.1942 -0.1368 0.6825 -0.6912>}
 dyn_frogLeg_04_r {parent=dyn_frogLeg_03_r, loc=<4.9329 0 0>, rot=<0.0811 0.0034 -0.2437 -0.9665>}
 dyn_frogLeg_05_r {parent=dyn_frogLeg_04_r, loc=<2.7609 0 0>, rot=<-0.6422 0.0355 0.7079 -0.2919>}
 dyn_wormAll_Base {parent=pelvis, loc=<2.586 13.2732 10.7737>, rot=<0.4478 -0.5431 0.5315 0.4712>}
 dyn_wormAll {parent=dyn_wormAll_Base, loc=<0 0 0>, rot=<0 0 0 -1>}
 dyn_wormB_01_Base {parent=dyn_wormAll, loc=<1.3938 -0.2191 -0.6567>, rot=<0 0 0 -1>}
 dyn_wormB_01 {parent=dyn_wormB_01_Base, loc=<0 0 0>, rot=<0 0 0 -1>}
 dyn_wormB_02 {parent=dyn_wormB_01, loc=<2.0166 0 0>, rot=<-0.002 -0.0055 -0.1863 -0.9825>}
 dyn_wormB_03 {parent=dyn_wormB_02, loc=<2.0791 0 0>, rot=<-0.1264 -0.1594 -0.1363 -0.9696>}
 dyn_wormB_04 {parent=dyn_wormB_03, loc=<1.8857 0 0>, rot=<-0.1001 -0.1589 0.2409 -0.9522>}
 dyn_wormB_05 {parent=dyn_wormB_04, loc=<1.9271 0 0>, rot=<-0.0011 -0.0562 0.2902 -0.9553>}
 dyn_wormB_06 {parent=dyn_wormB_05, loc=<2.6187 0 0>, rot=<0 0 0 -1>}
 dyn_wormC_01_Base {parent=dyn_wormAll, loc=<2.6393 1.3842 -0.6057>, rot=<-0.021 -0.0412 -0.3096 -0.9498>}
 dyn_wormC_01 {parent=dyn_wormC_01_Base, loc=<0 0 0>, rot=<0 0 0 -1>}
 dyn_wormC_02 {parent=dyn_wormC_01, loc=<2.7846 0 0>, rot=<-0.2984 -0.1724 -0.242 -0.907>}
 dyn_wormC_03 {parent=dyn_wormC_02, loc=<3.3105 0 0>, rot=<0.1607 0.0369 -0.0372 -0.9856>}
 dyn_wormC_04 {parent=dyn_wormC_03, loc=<2.3613 0 0>, rot=<0.1268 0.0536 0.2084 -0.9683>}
 dyn_wormC_05 {parent=dyn_wormC_04, loc=<2.6535 0 0>, rot=<0 0 0 -1>}
 dyn_wormA_01_Base {parent=dyn_wormAll, loc=<1.2994 -2.3542 -0.8412>, rot=<-0.0526 -0.1396 -0.2003 -0.9683>}
 dyn_wormA_01 {parent=dyn_wormA_01_Base, loc=<0 0 0>, rot=<0 0 0 1>}
 dyn_wormA_02 {parent=dyn_wormA_01, loc=<1.9609 0 0>, rot=<-0.2576 -0.6548 0.1336 -0.6979>}
 dyn_wormA_03 {parent=dyn_wormA_02, loc=<2.1813 0 0>, rot=<-0.0234 -0.3706 0.1135 -0.9216>}
 dyn_wormA_04 {parent=dyn_wormA_03, loc=<2.7252 0 0>, rot=<0.0218 -0.1117 0.132 -0.9847>}
 dyn_wormA_05 {parent=dyn_wormA_04, loc=<2.4219 0 0>, rot=<0.0404 -0.1016 0.052 -0.9926>}
 dyn_wormA_06 {parent=dyn_wormA_05, loc=<2.5484 0 0>, rot=<0 0 0 1>}
]
variables [targeting {type=double, default=0}]
