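# Physics asset for the face accessory of the medium male Freak player body, covering its dynamic hood, hood teeth and tongue bones.
# Capsule chains form the hood and an eight-segment tongue, while boxes form the hood teeth and left and right hood pieces, with angular limits along the chains.
# Each part's base joins spine_05 through free angular joints, alongside capsule bodies on the upper spine.
# The left and right hood pieces are angularly locked to their bases.
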PhysicsAsset /BRCosmetics/Characters/Player/Male/Medium/Bodies/M_MED_Freak/Meshes/Parts/M_MED_Freak_FaceAcc_Physics
bodies {
 dyn_hoodTeeth_01 [{shapeType=Box, extentX=8.4923, extentY=2.4693, extentZ=13.3647, center=<3.384 -0.0002 0>, rotation=<rot -9.44 0 0>}]
 dyn_hoodTeeth_01_Base [{shapeType=Box, extentX=3.2075, extentY=2.34, extentZ=7.0236, center=<3.3066 -0.0578 0.0225>, rotation=<rot -1.9763 -90 0.111>}]
 dyn_hoodTeeth_02 [{shapeType=Box, extentX=18.1038, extentY=5.3885, extentZ=11.5802, center=<1.2866 0.5995 -0.0655>, rotation=<rot 83.7626 90.0002 90.0003>}]
 dyn_hood_01 [{shapeType=Capsule, length=4.5448, radius=5.4298, center=<3.0934 0 0>, rotation=<rot 90 0 0>}]
 dyn_hood_01_Base [{shapeType=Capsule, length=0.505, radius=0.505, rotation=<rot 0 -90 0>}]
 dyn_hood_02 [{shapeType=Capsule, length=5.8174, radius=4.8942, center=<3.4715 0 0>, rotation=<rot 90 0 0>}]
 dyn_hood_03 [{shapeType=Capsule, length=5.6378, radius=4.4865, center=<2.1708 0 -0.0001>, rotation=<rot 90 0 0>}]
 dyn_hood_Base_l [{shapeType=Box, extentX=6.3243, extentY=6.9149, extentZ=11.3466, center=<5.4465 -0.6993 0.004>, rotation=<rot -0.0715 -90 -0.0686>}]
 dyn_hood_Base_r [{shapeType=Box, extentX=1.01, extentY=1.01, extentZ=1.01, rotation=<rot 0 -90 0>}]
 dyn_hood_l [{shapeType=Box, extentX=10.0509, extentY=4.1344, extentZ=8.3768, center=<4.0649 0.0002 0.232>, rotation=<rot 5.2732 0 0>}]
 dyn_hood_r [{shapeType=Box, extentX=10.5488, extentY=8.6045, extentZ=3.8226, center=<4.3805 0.7223 0>}]
 dyn_tongue_01 [{shapeType=Capsule, length=8.1354, radius=4.472, center=<3.5295 0 -0.0001>, rotation=<rot 90 0 0>}]
 dyn_tongue_01_Base [{shapeType=Capsule, length=0.505, radius=0.505, rotation=<rot 0 -90 0>}]
 dyn_tongue_02 [{shapeType=Capsule, length=4.4702, radius=3.9482, center=<3.9354 -0.829 0.0001>, rotation=<rot 90 0 0>}]
 dyn_tongue_03 [{shapeType=Capsule, length=2.5676, radius=4.7282, center=<3.6364 -1.2488 -0.0002>, rotation=<rot 90 0 5.8854>}]
 dyn_tongue_04 [{
   shapeType=Capsule
   length=3.5118
   radius=2
   shapeName=dyn_tongue_04_capsule
   center=<3.3066 -0.0578 0.0225>
   rotation=<rot -1.9763 -90 0.1111>
  }]
 dyn_tongue_05 [{
   shapeType=Capsule
   length=3.0315
   radius=2
   shapeName=dyn_tongue_05_capsule
   center=<3.301 -0.0249 0.0045>
   rotation=<rot -1.5655 -90 0.6665>
  }]
 dyn_tongue_06 [{
   shapeType=Capsule
   length=2.4301
   radius=1.5
   shapeName=dyn_tongue_06_capsule
   center=<2.1924 -0.0419 0.0079>
   rotation=<rot -0.7847 -90 2.1211>
  }]
 dyn_tongue_07 [{shapeType=Capsule, length=1.5, radius=1.54, shapeName=dyn_tongue_07_capsule, rotation=<rot 0 -90 0.0002>}]
 dyn_tongue_08 [{
   shapeType=Capsule
   length=1.8289
   radius=1.6186
   shapeName=dyn_tongue_08_capsule
   center=<-0.3722 -0.0039 0.0003>
   rotation=<rot 0 88.6244 -89.9896>
  }]
 spine_02 [{shapeType=Capsule, length=0.505, radius=0.505}]
 spine_03 [{shapeType=Capsule, length=18.3494, radius=11.9785, center=<0 4.4153 0>}]
 spine_04 [{shapeType=Capsule, length=0.0002, radius=12.5959, center=<-1.7796 2.908 0>}]
 spine_05 [{shapeType=Capsule, length=16.1814, radius=11.6179, center=<-3.8003 1.7845 0>}]
}
constraints [
 {bone1Name=spine_05, bone2Name=spine_04, swing1LimitDegrees=45, swing2LimitDegrees=45, twistLimitDegrees=45}
 {bone1Name=dyn_hood_01, bone2Name=dyn_hood_01_Base, swing1LimitDegrees=8, swing2LimitDegrees=8, twistLimitDegrees=4}
 {bone1Name=dyn_tongue_08, bone2Name=dyn_tongue_07, swing1LimitDegrees=30, swing2LimitDegrees=20, twistLimitDegrees=16}
 {bone1Name=dyn_tongue_07, bone2Name=dyn_tongue_06, swing1LimitDegrees=30, swing2LimitDegrees=20, twistLimitDegrees=16}
 {bone1Name=dyn_tongue_06, bone2Name=dyn_tongue_05, swing1LimitDegrees=30, swing2LimitDegrees=20, twistLimitDegrees=16}
 {bone1Name=dyn_tongue_05, bone2Name=dyn_tongue_04, swing1LimitDegrees=30, swing2LimitDegrees=20, twistLimitDegrees=16}
 {bone1Name=dyn_tongue_04, bone2Name=dyn_tongue_03, swing1LimitDegrees=30, swing2LimitDegrees=20, twistLimitDegrees=16}
 {bone1Name=dyn_tongue_03, bone2Name=dyn_tongue_02, swing1LimitDegrees=30, swing2LimitDegrees=20, twistLimitDegrees=16}
 {
  bone1Name=spine_03
  bone2Name=spine_02
  swing1LimitDegrees=45
  swing1Motion=Free
  swing2LimitDegrees=45
  swing2Motion=Free
  twistLimitDegrees=45
  twistMotion=Free
 }
 {
  bone1Name=spine_04
  bone2Name=spine_03
  swing1LimitDegrees=45
  swing1Motion=Free
  swing2LimitDegrees=45
  swing2Motion=Free
  twistLimitDegrees=45
  twistMotion=Free
 }
 {bone1Name=dyn_hood_02, bone2Name=dyn_hood_01, swing1LimitDegrees=14, swing2LimitDegrees=14, twistLimitDegrees=6}
 {bone1Name=dyn_hood_03, bone2Name=dyn_hood_02, swing1LimitDegrees=16, swing2LimitDegrees=16, twistLimitDegrees=8}
 {
  bone1Name=dyn_hood_01_Base
  bone2Name=spine_05
  swing1LimitDegrees=45
  swing1Motion=Free
  swing2LimitDegrees=45
  swing2Motion=Free
  twistLimitDegrees=45
  twistMotion=Free
 }
 {bone1Name=dyn_tongue_02, bone2Name=dyn_tongue_01, swing1LimitDegrees=30, swing2LimitDegrees=20, twistLimitDegrees=16}
 {
  bone1Name=dyn_tongue_01_Base
  bone2Name=spine_05
  swing1LimitDegrees=45
  swing1Motion=Free
  swing2LimitDegrees=45
  swing2Motion=Free
  twistLimitDegrees=45
  twistMotion=Free
 }
 {bone1Name=dyn_tongue_01, bone2Name=dyn_tongue_01_Base, swing1LimitDegrees=30, swing2LimitDegrees=20, twistLimitDegrees=16}
 {
  bone1Name=dyn_hood_l
  bone2Name=dyn_hood_Base_l
  swing1LimitDegrees=20
  swing1Motion=Locked
  swing2LimitDegrees=12
  swing2Motion=Locked
  twistLimitDegrees=8
  twistMotion=Locked
 }
 {
  bone1Name=dyn_hood_r
  bone2Name=dyn_hood_Base_r
  swing1LimitDegrees=12
  swing1Motion=Locked
  swing2LimitDegrees=20
  swing2Motion=Locked
  twistLimitDegrees=8
  twistMotion=Locked
 }
 {bone1Name=dyn_hoodTeeth_01, bone2Name=dyn_hoodTeeth_01_Base, swing1LimitDegrees=10, swing2LimitDegrees=10, twistLimitDegrees=8}
 {bone1Name=dyn_hoodTeeth_02, bone2Name=dyn_hoodTeeth_01, swing1LimitDegrees=10, swing2LimitDegrees=10, twistLimitDegrees=8}
 {
  bone1Name=dyn_hoodTeeth_01_Base
  bone2Name=spine_05
  swing1LimitDegrees=45
  swing1Motion=Free
  swing2LimitDegrees=45
  swing2Motion=Free
  twistLimitDegrees=45
  twistMotion=Free
 }
 {
  bone1Name=dyn_hood_Base_l
  bone2Name=spine_05
  swing1LimitDegrees=45
  swing1Motion=Free
  swing2LimitDegrees=45
  swing2Motion=Free
  twistLimitDegrees=45
  twistMotion=Free
 }
 {
  bone1Name=dyn_hood_Base_r
  bone2Name=spine_05
  swing1LimitDegrees=45
  swing1Motion=Free
  swing2LimitDegrees=45
  swing2Motion=Free
  twistLimitDegrees=45
  twistMotion=Free
 }
]
