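# Animation Blueprint for the face accessory part of the M_MED_Freak male cosmetic character, adding physics and aim-time correction.
# It copies the attached parent's pose and curves, then runs RigidBody simulation relative to spine_05.
# An additive head pitch follows, weighted while targeting by neck_01's yaw offset from reference pose.
# A Control Rig finishes the pose, fed a targeting value that eases in faster than out.
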AnimBlueprint /BRCosmetics/Characters/Player/Male/Medium/Bodies/M_MED_Freak/Meshes/Parts/M_MED_Freak_FaceAcc_AnimBP
target_skeleton=&/M_MED_Freak_FaceAcc_Skeleton
graphs [
 AnimGraph {nodes [
   Component To Local(ComponentPose=Local To Component())
   n1 = RigidBody(ComponentPose=Local To Component(LocalPose=Copy Pose From Mesh() {bCopyCurves=true, bUseAttachedParent=true})) {
    bClampLinearTranslationLimitToRefPose=true
    bOverrideWorldGravity=true
    overridePhysicsAsset=&/M_MED_Freak_FaceAcc_Physics
    simulationSpace=BaseBoneSpace
    alphaBoolBlend {blendOption=Linear}
    baseBoneRef {boneName=spine_05}
    componentAppliedLinearAccClamp {z=4000}
    componentLinearAccScale=<1 1 0.8>
    componentLinearVelScale=<0.7 0.7 0.5>
    overrideWorldGravity {z=-980}
   }
   n2 = "Transform (Modify) Bone - Bone: head"(ComponentPose=@n1, Alpha=Get ADS() -> (ADS:double), Translation="0,0,0", Rotation="22,0.000,0", Scale="1,1,1") {
    rotationMode=BMM_Additive
    rotationSpace=BCS_BoneSpace
    translationMode=BMM_Additive
    translationSpace=BCS_BoneSpace
    alphaBoolBlend {blendOption=Linear}
    alphaScaleBiasClamp {interpSpeedDecreasing=20, interpSpeedIncreasing=20}
    boneToModify {boneName=head}
    rotation {pitch=22}
   }
   n3 = Control Rig!(Source=Component To Local(ComponentPose=@n2), Alpha, targeting:double=Get SwingDelta() -> (SwingDelta:double)) -> (Pose) {
    alphaBoolBlend {blendOption=Linear}
    controlRigAssetReference {blueprintRigClass=&/M_MED_Freak_FaceAcc_CtrlRig.M_MED_Freak_FaceAcc_CtrlRig_C}
   }
   Output Pose(Result=@n3) {layerGroup=None}
  ]}
 EventGraph {nodes [
   n1 = Event BlueprintUpdateAnimation()
   > n2 = Sequence()
   n3 = Get bIsTargeting()
   > n4 = Branch(Condition=@n3)
   > Set SwingDelta(SwingDelta:double=FInterpTo(Current=Get SwingDelta(), Target=1, DeltaTime=@n1.DeltaTimeX, InterpSpeed=5)) -> (Output_Get:double)
   Set SwingDelta(execute=@n4.else, SwingDelta=FInterpTo(Current=Get SwingDelta(), Target=0, DeltaTime=@n1.DeltaTimeX, InterpSpeed=2))
   n5 = Branch(execute=@n2.then_1, Condition=Get bIsTargeting())
   n6 = BreakTransform(InTransform=GetDeltaTransformFromRefPose(self=GetOwnerSkeletalMeshComponent(), BoneName=neck_01, BaseName=None))
   > Set ADS(ADS:double=MapRangeClamped(Value=BreakRotator(InRot=@n6.Rotation).Yaw, InRangeA=0, InRangeB=-10)) -> (Output_Get:double)
   Set ADS(execute=@n5.else, ADS=0)
   TryGetPawnOwner()
   FInterpTo(Current=0, Target="To Float (Boolean)"(InBool=@n3), DeltaTime=@n1.DeltaTimeX, InterpSpeed=2)
  ]}
]
variables [
 ADS {type=number}
 SwingDelta {type=number}
]
